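# Material function for a high-quality desert sand landscape layer in the Helios environment, apparently adapted from an asphalt layer given its leftover names.
# It samples wavy sand textures at two world-UV scales, blending color and normals by a height lerp driven by squared layer weight and variation noise.
# Vertex-normal slope blends two tints for base color.
# Material attributes output also carries normal, specular and roughness from one mask channel, opacity mask and displacement.
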MaterialFunction /Game/Athena/Helios/Environments/Landscape/Materials/CH5S3_PROTO/MF_Helios_Desert_HQ
graph {nodes [
  n1 = FunctionInput(Preview=Constant()) {bUsePreviewValueAsDefault=true, inputName=LayerWeight, inputType=FunctionInput_Scalar, sortPriority=-500}
  n2 = MaterialFunctionCall!(
   Coordinates
   R_01_Scale=ScalarParameter() {defaultValue=10000, parameterName=SandVariationNoise_Scale}
   R_02_Scale
   R_03_Scale
   R_04_Scale
   G_01_Scale
   G_02_Scale
   G_03_Scale
   G_04_Scale
  ) {materialFunction=/Game/Athena/Helios/Environments/Landscape/Materials/MaterialFunctions_Utilities/MF_VariationMasks2}
  n3 = MaterialFunctionCall(Transition Phase=Multiply(A=Multiply(A=@n1, B=@n1), B=OneMinus(None=@n2.R_01)), Height Texture=NamedRerouteUsage()) {materialFunction=&HeightLerp}
  n4 = FunctionInput(Preview=StaticBool() {value=true}) {
   bUsePreviewValueAsDefault=true
   description="Default = True"
   inputName=UseTangentNormals
   inputType=FunctionInput_StaticBool
   sortPriority=1
  }
  n5 = MaterialFunctionCall!(
   Normals=LinearInterpolate(A=NamedRerouteUsage(), B=NamedRerouteUsage(), Alpha=@n3.Lerp Alpha No Contrast)
   TransformToTangentSpace=@n4
   Normal Strength
  ) {materialFunction=/Game/Athena/Apollo/Environments/Landscape/Materials/Biome/CommonMaterials/MF_HQ_WorldAlignedNormal}
  Fresnel(Normal=@n5.Result)
  n6 = MaterialFunctionCall!(Scale=NamedRerouteUsage()) {materialFunction=/Game/Athena/Apollo/Environments/Landscape/Materials/Biome/CommonMaterials/MF_SimpleWorldUV}
  n7 = TextureSample(UVs=@n6.Result) {samplerSource=SSM_Wrap_WorldGroupSettings, texture=/Game/Environments/Artemis/Sets/Terrain/Textures/T_Artemis_Desert_SandWavy_02_D}
  n8 = LinearInterpolate(A=NamedRerouteUsage(), B=@n7.RGB, Alpha=@n3.Lerp Alpha No Contrast)
  n9 = Multiply(A=VectorParameter() {parameterName=Desert_HQ_Tint2, defaultValue=<rgba 0.01 0.0059 0.0036 1>}.RGB, B=@n8)
  n10 = Multiply(A=VectorParameter() {parameterName=Desert_HQ_Tint, defaultValue=<rgba 1 0.5942 0.3597 1>}.RGB, B=@n8)
  n11 = ComponentMask(None=VertexNormalWS()) {b=true}
  n12 = Power(Base=OneMinus(None=Abs(None=@n11)), Exp=ScalarParameter() {defaultValue=3, parameterName=Desert_HQ_Tint_Slope})
  n13 = TextureSample(UVs=@n6.Result) {
   samplerSource=SSM_Wrap_WorldGroupSettings
   samplerType=SAMPLERTYPE_Masks
   texture=/Game/Environments/Artemis/Sets/Terrain/Textures/T_Artemis_Desert_SandWavy_02_S
  }
  n14 = MaterialFunctionCall(Float3=@n13.RGB) {materialFunction=&BreakOutFloat3Components}
  n15 = NamedRerouteUsage()
  n16 = SetMaterialAttributes!(
   MaterialAttributes
   BaseColor=LinearInterpolate(A=Saturate(None=@n9), B=Saturate(None=@n10), Alpha=Saturate(None=@n12))
   Specular=@n14.R
   Roughness=@n14.R
   Normal=@n5.Result
   OpacityMask=@n15
   Displacement=LinearInterpolate(B=@n15, Alpha=ScalarParameter() {group=Displacement, parameterName=Asphalt_DisplacementAmt}) {constA=0.5}
  ) {attributeSetTypes [
    #1
    #2
    #3
    #4
    #5
    #6
   ]}
  FunctionOutput(None=@n16)
  MaterialFunctionCall!(
   Texture
   LandscapeCoordinates
   LowEndTiling
   HexGridGradientSharpness
   HexGridSize
   HexGridUVInput
   MinScale
   MaxScale
   MinOffset
   MaxOffset
   "MaxRotation(Degrees)"
   "MinRotation(Degrees)"
   forceLowQuality
  ) {materialFunction=/Game/Athena/Asteria/Environments/Landscape/Materials/MaterialFunctions/UtilityFunctions/Landscape_TextureSample}
  MaterialFunctionCall!(Tiling, Texture, LandscapeCoordinates) {materialFunction=/Game/Athena/Asteria/Environments/Landscape/Materials/MaterialFunctions/UtilityFunctions/MF_LandscapeTextureSample}
  Multiply(A=NamedRerouteUsage(), B=@n7.RGB)
  NamedRerouteDeclaration(None=Constant() {r=1000}) {name=CrackScale, nodeColor=<rgba 0.5469 0 1 1>}
  n17 = MaterialFunctionCall(Scale=NamedRerouteUsage()) {materialFunction=&MF_SimpleWorldUV}
  n18 = TextureSample(UVs=@n17.Result) {samplerSource=SSM_Wrap_WorldGroupSettings, texture=&T_Artemis_Desert_SandWavy_02_D}
  NamedRerouteDeclaration(None=@n18.RGB) {name=CrackDiffuse, nodeColor=<rgba 0.5469 0 1 1>}
  NamedRerouteDeclaration(None=Constant() {r=350}) {name=AsphaltScale, nodeColor=<rgba 0.5469 0 1 1>}
  NamedRerouteDeclaration(None=Constant() {r=0.5905}) {name=AsphaltVariation, nodeColor=<rgba 0.5469 0 1 1>}
  NamedRerouteDeclaration(None=MaterialFunctionCall(Transition Phase=NamedRerouteUsage(), Height Texture=@n7.A) {materialFunction=&HeightLerp}.Lerp Alpha No Contrast) {name=Asphalt_Height, nodeColor=<rgba 0.5469 0 1 1>}
  n19 = TextureSample(UVs=@n6.Result) {
   samplerSource=SSM_Wrap_WorldGroupSettings
   samplerType=SAMPLERTYPE_Normal
   texture=/Game/Environments/Artemis/Sets/Terrain/Textures/T_Artemis_Desert_SandWavy_02_N
  }
  NamedRerouteDeclaration(None=@n19.RGB) {name=AsphaltNormal, nodeColor=<rgba 0.5469 0 1 1>}
  NamedRerouteDeclaration(None=@n18.A) {name=CrackHeight, nodeColor=<rgba 0.5469 0 1 1>}
  n20 = TextureSample(UVs=@n17.Result) {
   samplerSource=SSM_Wrap_WorldGroupSettings
   samplerType=SAMPLERTYPE_Normal
   texture=&T_Artemis_Desert_SandWavy_02_N
  }
  NamedRerouteDeclaration(None=@n20.RGB) {name=CrackNormal, nodeColor=<rgba 0.5469 0 1 1>}
  n21 = TextureSample(UVs=@n17.Result) {
   samplerSource=SSM_Wrap_WorldGroupSettings
   samplerType=SAMPLERTYPE_Masks
   texture=&T_Artemis_Desert_SandWavy_02_S
  }
  OneMinus(None=Multiply(A=NamedRerouteUsage(), B=ComponentMask(None=@n21.RGB) {r=true}))
  ScalarParameter() {defaultValue=0.5905, group=Asphalt, parameterName=AsphaltVariation}
  ScalarParameter() {defaultValue=350, group=Asphalt, parameterName=AsphaltMainScale}
  ScalarParameter() {parameterName=SandFresnelExponent}
  ScalarParameter() {defaultValue=2000, group=Asphalt, parameterName=SizeCracks}
 ]}
