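# High-quality Wasteland landscape layer material function for Helios, blending world-UV gravel and Sand3 textures into a Material output.
# Squared LayerWeight and inverted variation noise drive a HeightLerp that mixes color and world-aligned normals; gravel masks supply specular and roughness.
# The base color is then desaturated and tinted by wasteland color parameters.
# Asphalt/crack names label the sand/gravel reroutes, and several nodes, including the biome tint call, are unconnected to the output.
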MaterialFunction /Game/Athena/Helios/Environments/Landscape/Materials/CH5S3_PROTO/MF_Helios_Wasteland_HQ
graph {
 properties {description=-3909}
 nodes [
  Constant3Vector() {constant=<rgba 0.4735 0.2582 0.1248 1>}
  Constant3Vector() {constant=<rgba 0.3231 0 0.7913 1>}
  Constant3Vector() {constant=<rgba 0.0091 0.0648 1 1>}
  n1 = FunctionInput(Preview=Constant()) {bUsePreviewValueAsDefault=true, inputName=LayerWeight, inputType=FunctionInput_Scalar, sortPriority=-500}
  n2 = MaterialFunctionCall!(
   Coordinates
   R_01_Scale=ScalarParameter() {defaultValue=10000, parameterName=WastelandVariationNoise_Scale}
   R_02_Scale
   R_03_Scale
   R_04_Scale
   G_01_Scale
   G_02_Scale
   G_03_Scale
   G_04_Scale
  ) {materialFunction=/Game/Athena/Helios/Environments/Landscape/Materials/MaterialFunctions_Utilities/MF_VariationMasks2}
  n3 = MaterialFunctionCall(Transition Phase=Multiply(A=Multiply(A=@n1, B=@n1), B=OneMinus(None=@n2.R_01)), Height Texture=NamedRerouteUsage()) {materialFunction=&HeightLerp}
  n4 = FunctionInput(Preview=StaticBool() {value=true}) {
   bUsePreviewValueAsDefault=true
   description="Default = True"
   inputName=UseTangentNormals
   inputType=FunctionInput_StaticBool
   sortPriority=1
  }
  n5 = MaterialFunctionCall!(
   Normals=LinearInterpolate(A=NamedRerouteUsage(), B=NamedRerouteUsage(), Alpha=@n3.Lerp Alpha No Contrast)
   TransformToTangentSpace=@n4
   Normal Strength
  ) {materialFunction=/Game/Athena/Apollo/Environments/Landscape/Materials/Biome/CommonMaterials/MF_HQ_WorldAlignedNormal}
  Fresnel(Normal=@n5.Result)
  n6 = MaterialFunctionCall!(Scale=NamedRerouteUsage()) {materialFunction=/Game/Athena/Apollo/Environments/Landscape/Materials/Biome/CommonMaterials/MF_SimpleWorldUV}
  n7 = TextureSample(UVs=@n6.Result) {samplerSource=SSM_Wrap_WorldGroupSettings, texture=&/Sand3_D}
  n8 = TextureSample(UVs=@n6.Result) {
   samplerSource=SSM_Wrap_WorldGroupSettings
   samplerType=SAMPLERTYPE_Masks
   texture=/Game/Environments/Helios/Terrain/Textures/T_Helios_Gravel_01_S
  }
  n9 = MaterialFunctionCall(Float3=@n8.RGB) {materialFunction=&BreakOutFloat3Components}
  n10 = NamedRerouteUsage()
  n11 = SetMaterialAttributes!(
   MaterialAttributes
   BaseColor=LinearInterpolate(A=NamedRerouteUsage(), B=@n7.RGB, Alpha=@n3.Lerp Alpha No Contrast)
   Specular=@n9.R
   Roughness=@n9.R
   Normal=@n5.Result
   OpacityMask=@n10
   Displacement=LinearInterpolate(B=@n10, Alpha=ScalarParameter() {group=Displacement, parameterName=Asphalt_DisplacementAmt}) {constA=0.5}
  ) {attributeSetTypes [
    #1
    #2
    #3
    #4
    #5
    #6
   ]}
  n12 = GetMaterialAttributes(None=@n11) {attributeGetTypes [#1]}
  n13 = Multiply(
   A=Desaturation(None=@n12.BaseColor, Fraction=ScalarParameter() {defaultValue=0.5, parameterName=DesatFactor})
   B=VectorParameter() {parameterName=WasteLandColor, defaultValue=<rgba 0.7552 0.5299 0.375 1>}.RGB
  )
  n14 = ScalarParameter() {defaultValue=1, parameterName=WastelandColorValue}
  FunctionOutput(None=SetMaterialAttributes!(MaterialAttributes=@n12.MaterialAttributes, BaseColor=Multiply(A=@n13, B=@n14)) {attributeSetTypes [#1]}) {outputName=Material}
  n15 = Constant()
  n16 = Constant()
  n17 = Constant()
  n18 = Constant()
  MaterialFunctionCall!(
   Material
   TerrainMasks_QuadA=NamedRerouteDeclaration(None=FunctionInput(Preview=@n15) {bUsePreviewValueAsDefault=true, inputName=TerrainMasks_QuadA, inputType=FunctionInput_Vector4, sortPriority=-400}) {name=TerrainMasks_QuadA, nodeColor=<rgba 1 0 1 1>}
   TerrainMasks_QuadB=NamedRerouteDeclaration(None=FunctionInput(Preview=@n16) {bUsePreviewValueAsDefault=true, inputName=TerrainMasks_QuadB, inputType=FunctionInput_Vector4, sortPriority=-300}) {name=TerrainMasks_QuadB, nodeColor=<rgba 1 0 1 1>}
   TerrainMasks_QuadC=NamedRerouteDeclaration(None=FunctionInput(Preview=@n17) {bUsePreviewValueAsDefault=true, inputName=TerrainMasks_QuadC, inputType=FunctionInput_Vector4, sortPriority=-200}) {name=TerrainMasks_QuadC, nodeColor=<rgba 1 0 1 1>}
   TerrainMasks_QuadD=NamedRerouteDeclaration(None=FunctionInput(Preview=@n18) {bUsePreviewValueAsDefault=true, inputName=TerrainMasks_QuadD, inputType=FunctionInput_Vector4, sortPriority=-100}) {name=TerrainMasks_QuadD, nodeColor=<rgba 1 0 1 1>}
   OD_MD_SedimentLayers
  ) {materialFunction=/Game/Athena/Helios/Environments/Landscape/Materials/MaterialFunctions_Utilities/MF_Sand_BiomeColorTint}
  MaterialFunctionCall!(
   Texture
   LandscapeCoordinates
   LowEndTiling
   HexGridGradientSharpness
   HexGridSize
   HexGridUVInput
   MinScale
   MaxScale
   MinOffset
   MaxOffset
   "MaxRotation(Degrees)"
   "MinRotation(Degrees)"
   forceLowQuality
  ) {materialFunction=/Game/Athena/Asteria/Environments/Landscape/Materials/MaterialFunctions/UtilityFunctions/Landscape_TextureSample}
  MaterialFunctionCall!(Tiling, Texture, LandscapeCoordinates) {materialFunction=/Game/Athena/Asteria/Environments/Landscape/Materials/MaterialFunctions/UtilityFunctions/MF_LandscapeTextureSample}
  Multiply(A=NamedRerouteUsage(), B=@n7.RGB)
  n19 = MaterialFunctionCall(Scale=NamedRerouteUsage()) {materialFunction=&MF_SimpleWorldUV}
  n20 = TextureSample(UVs=@n19.Result) {samplerSource=SSM_Wrap_WorldGroupSettings, texture=/Game/Environments/Helios/Terrain/Textures/T_Helios_Gravel_01_D}
  NamedRerouteDeclaration(None=@n20.A) {name=CrackHeight, nodeColor=<rgba 0.5469 0 1 1>}
  NamedRerouteDeclaration(None=MaterialFunctionCall(Transition Phase=NamedRerouteUsage(), Height Texture=@n7.A) {materialFunction=&HeightLerp}.Lerp Alpha No Contrast) {name=Asphalt_Height, nodeColor=<rgba 0.5469 0 1 1>}
  NamedRerouteDeclaration(None=Constant() {r=0.5905}) {name=AsphaltVariation, nodeColor=<rgba 0.5469 0 1 1>}
  NamedRerouteDeclaration(None=Constant() {r=1000}) {name=AsphaltScale, nodeColor=<rgba 0.5469 0 1 1>}
  NamedRerouteDeclaration(None=Constant() {r=1000}) {name=CrackScale, nodeColor=<rgba 0.5469 0 1 1>}
  NamedRerouteDeclaration(None=@n20.RGB) {name=CrackDiffuse, nodeColor=<rgba 0.5469 0 1 1>}
  n21 = TextureSample(UVs=@n19.Result) {
   samplerSource=SSM_Wrap_WorldGroupSettings
   samplerType=SAMPLERTYPE_Normal
   texture=/Game/Environments/Helios/Terrain/Textures/T_Helios_Gravel_01_N
  }
  NamedRerouteDeclaration(None=@n21.RGB) {name=CrackNormal, nodeColor=<rgba 0.5469 0 1 1>}
  n22 = TextureSample(UVs=@n6.Result) {
   samplerSource=SSM_Wrap_WorldGroupSettings
   samplerType=SAMPLERTYPE_Normal
   texture=&/Sand3_N
  }
  NamedRerouteDeclaration(None=@n22.RGB) {name=AsphaltNormal, nodeColor=<rgba 0.5469 0 1 1>}
  n23 = TextureSample(UVs=@n19.Result) {
   samplerSource=SSM_Wrap_WorldGroupSettings
   samplerType=SAMPLERTYPE_Masks
   texture=&T_Helios_Gravel_01_S
  }
  OneMinus(None=Multiply(A=NamedRerouteUsage(), B=ComponentMask(None=@n23.RGB) {r=true}))
  ScalarParameter() {defaultValue=350, group=Asphalt, parameterName=AsphaltMainScale}
  ScalarParameter() {defaultValue=0.5905, group=Asphalt, parameterName=AsphaltVariation}
  ScalarParameter() {defaultValue=2000, group=Asphalt, parameterName=SizeCracks}
  ScalarParameter() {parameterName=SandFresnelExponent}
 ]
}
