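# Material function that builds a sand surface for the Helios Wasteland landscape, using mobile sand parameters and one packed mask texture.
# A static switch selects world-space or landscape coordinates for texture sampling and for the height-based SandOceanMask.
# Named reroutes carry noise-driven color blends, specular, roughness and faded normals.
# The resulting material attributes pass through MF_Sand_BiomeColorTint with the BiomeMask input before output.
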MaterialFunction /Game/Athena/Helios/Environments/Landscape/Materials/CH5S3_PROTO/MF_Helios_Wasteland_NS
graph {nodes [
  n1 = SetMaterialAttributes!(
   MaterialAttributes
   BaseColor=NamedRerouteUsage()
   Specular=NamedRerouteUsage()
   Roughness=NamedRerouteUsage()
   Normal=NamedRerouteUsage()
   OpacityMask=NamedRerouteUsage()
  ) {
   desc=HighEnd Material
   attributeSetTypes [
    #1
    #2
    #3
    #4
    #5
   ]
  }
  FunctionOutput(None=MaterialFunctionCall!(
    Material=@n1
    TerrainMasks_QuadA
    TerrainMasks_QuadB=FunctionInput() {bUsePreviewValueAsDefault=true, inputName=BiomeMask, inputType=FunctionInput_Vector4}
    TerrainMasks_QuadC
    TerrainMasks_QuadD
    OD_MD_SedimentLayers
   ) {materialFunction=/Game/Athena/Helios/Environments/Landscape/Materials/MaterialFunctions_Utilities/MF_Sand_BiomeColorTint}.Result)
  n2 = Divide(
   A=WorldPosition() {worldPositionShaderOffset=WPT_ExcludeAllShaderOffsets}.XY
   B=ScalarParameter() {defaultValue=600, group=Mobile Sand, parameterName=Mobile Sand Scale, sortPriority=-1000}
  )
  n3 = LandscapeLayerCoords() {mappingScale=4.5}
  n4 = ComponentMask(None=LandscapeLayerCoords() {mappingPanV=7, mappingScale=-1, mappingType=TCMT_XZ}) {g=true}
  n5 = TextureSample(UVs=Multiply(
    A=StaticSwitchParameter(True=@n2, False=@n3) {defaultValue=true, group=Mobile Static Switch, parameterName=UseWorldSpaceCoords}
    B=Saturate(None=@n4)
   )) {samplerType=SAMPLERTYPE_Masks, texture=/Game/Environments/Helios/Terrain/TexturesMobile/T_Helios_Mobile_Sand}
  NamedRerouteDeclaration(None=@n5.RGBA) {name=Texture, nodeColor=<rgba 1 0.8672 0 1>}
  NamedRerouteDeclaration(None=ScalarParameter() {defaultValue=0.65, group=Mobile Sand, parameterName=Mobile Sand Roughness}) {name=Roughness, nodeColor=<rgba 0.1484 0 1 1>}
  n6 = ComponentMask(None=NamedRerouteUsage()) {b=true}
  n7 = Saturate(None=Subtract(A=Multiply(A=@n6) {constB=2}) {constB=1.5})
  NamedRerouteDeclaration(None=LinearInterpolate(
    A=ScalarParameter() {defaultValue=0.25, group=Mobile Sand, parameterName=Mobile Sand Specular Min}
    B=ScalarParameter() {defaultValue=0.6, group=Mobile Sand, parameterName=Mobile Sand Specular Max}
    Alpha=NamedRerouteDeclaration(None=@n7) {name=Sand Noise, nodeColor=<rgba 1 0.8672 0 1>}
   )) {name=Specular, nodeColor=<rgba 0.5 0 1 1>}
  n8 = NamedRerouteUsage()
  n9 = Frac(None=Multiply(A=ComponentMask(None=@n8) {b=true, g=true, r=true}) {constB=6})
  n10 = ComponentMask(None=NamedRerouteUsage()) {b=true}
  n11 = LinearInterpolate(
   A=LinearInterpolate(
    A=VectorParameter() {group=Mobile Sand, parameterName=Mobile Sand Color 1, defaultValue=<rgba 0.526 0.3211 0.1534 1>}.RGB
    B=VectorParameter() {group=Mobile Sand, parameterName=Mobile Sand Color 2, defaultValue=<rgba 0.7812 0.4967 0.2726 1>}.RGB
    Alpha=NamedRerouteUsage()
   )
   B=LinearInterpolate(
    A=VectorParameter() {group=Mobile Sand, parameterName=Mobile Sand Color 3, defaultValue=<rgba 0.3278 0.2346 0.1559 1>}.RGB
    B=Step(Y=@n9) {constX=0.51}
    Alpha=ScalarParameter() {defaultValue=0.065, group=Mobile Sand, parameterName=Mobile Sand Color Variation}
   ) {constA=0.5}
   Alpha=Saturate(None=LinearInterpolate(Alpha=@n10) {constA=11, constB=-3})
  )
  NamedRerouteDeclaration(None=LinearInterpolate(A=Multiply(A=@n11) {constB=0.6}, B=@n11, Alpha=NamedRerouteUsage())) {name=BaseColor, nodeColor=<rgba 0.5 0 1 1>}
  NamedRerouteDeclaration(None=FunctionInput() {bUsePreviewValueAsDefault=true, inputName=VariationMasks Mobile, inputType=FunctionInput_Vector4, sortPriority=3}) {name=VariationMasks Mobile, nodeColor=<rgba 0.9219 1 0 1>}
  n12 = NamedRerouteUsage()
  n13 = Add(A=Multiply(A=NamedRerouteUsage()) {constB=2}) {constB=-0.5}
  NamedRerouteDeclaration(None=LinearInterpolate(B=ComponentMask(None=@n12) {b=true, g=true, r=true}, Alpha=Saturate(None=@n13)) {constA=0.5}) {name=Normals, nodeColor=<rgba 0.5 0 1 1>}
  n14 = Multiply(A=Add(A=WorldPosition().Z) {constB=70}) {constB=0.01}
  n15 = LandscapeLayerCoords() {mappingPanV=1.5, mappingScale=0.5, mappingType=TCMT_XZ}
  n16 = Add(
   A=StaticSwitchParameter(True=Saturate(None=@n14), False=ComponentMask(None=@n15) {g=true}) {defaultValue=true, group=Mobile Static Switch, parameterName=UseWorldSpaceCoords}
   B=NamedRerouteUsage()
  )
  NamedRerouteDeclaration(None=Saturate(None=Subtract(A=@n16))) {name=SandOceanMask, nodeColor=<rgba 0 0.1094 1 1>}
  NamedRerouteDeclaration(None=@n5.A) {name=HeightMapFromTextures, nodeColor=<rgba 0.5 0 1 1>}
 ]}
