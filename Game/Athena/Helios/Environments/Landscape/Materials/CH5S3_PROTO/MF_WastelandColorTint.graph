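# Material function that applies a wasteland color tint to a material's base color, stored with the Helios landscape prototype materials.
# It desaturates BaseColor, then multiplies it by a tint color and a value multiplier, with each input defaulting to a shared WastelandColorTint parameter.
# A layer weight input lerps between original and tinted color, and the other material attributes pass through unchanged.
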MaterialFunction /Game/Athena/Helios/Environments/Landscape/Materials/CH5S3_PROTO/MF_WastelandColorTint
graph {nodes [
  n1 = GetMaterialAttributes(None=FunctionInput() {inputName=Material, inputType=FunctionInput_MaterialAttributes}) {attributeGetTypes [#1]}
  n2 = FunctionInput(Preview=ScalarParameter() {defaultValue=0.5, group=WastelandColorTint, parameterName=DesatFactor}) {bUsePreviewValueAsDefault=true, inputName=DesatFactor, inputType=FunctionInput_Scalar}
  n3 = VectorParameter() {group=WastelandColorTint, parameterName=WasteLandColor, defaultValue=<rgba 0.7552 0.5299 0.375 1>}
  n4 = Multiply(
   A=Multiply(A=Desaturation(None=@n1.BaseColor, Fraction=@n2), B=FunctionInput(Preview=@n3.RGB) {bUsePreviewValueAsDefault=true, inputName=Color})
   B=FunctionInput(Preview=ScalarParameter() {defaultValue=1, group=WastelandColorTint, parameterName=WastelandColorValue}) {bUsePreviewValueAsDefault=true, inputName=Value, inputType=FunctionInput_Scalar}
  )
  n5 = FunctionInput() {inputName=LayerWEight, inputType=FunctionInput_Scalar}
  FunctionOutput(None=SetMaterialAttributes!(MaterialAttributes=@n1.MaterialAttributes, BaseColor=LinearInterpolate(A=@n1.BaseColor, B=@n4, Alpha=@n5)) {attributeSetTypes [#1]}) {outputName=Material}
 ]}
