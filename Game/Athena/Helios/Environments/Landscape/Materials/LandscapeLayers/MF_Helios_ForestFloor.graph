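# Helios landscape layer function that adds a forest floor, under the BiomeA layer switch, to the incoming terrain material.
# World-UV forest floor textures with displacement optionally blend with Doomstadt leaves through a masked height lerp, then receive a biome color tint from terrain masks.
# A height-blend function composites the result by layer weight.
# MobileBlendedMaterial bypasses that blend; a separate mobile forest blend supplies a mobile layer mask.
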MaterialFunction /Game/Athena/Helios/Environments/Landscape/Materials/LandscapeLayers/MF_Helios_ForestFloor
graph {nodes [
  ComponentMask() {g=true, r=true}
  n1 = Constant()
  n2 = ComponentMask(None=NamedRerouteDeclaration(None=FunctionInput(Preview=@n1) {bUsePreviewValueAsDefault=true, inputName=TerrainMasks_QuadC, inputType=FunctionInput_Vector4, sortPriority=30}) {name=TerrainMasks_QuadC, nodeColor=<rgba 0.5 0 1 1>}) {a=true}
  n3 = NamedRerouteUsage()
  Composite!(DoomstadtTrees_Mask=@n2, LayerWeight=@n3) {subgraphName=ForestFloor_High}
  Constant() {r=1}
  Constant() {r=2}
  Constant() {r=2}
  n4 = NamedRerouteUsage()
  n5 = NamedRerouteUsage()
  n6 = NamedRerouteUsage()
  n7 = NamedRerouteUsage()
  n8 = NamedRerouteUsage()
  n9 = MaterialFunctionCall!(InputMaterial=@n4, LayerMaterial=@n5, LayerWeight=@n6, UseContrast=@n7, Contrast=@n8) {materialFunction=/Game/Athena/Helios/Environments/Landscape/Materials/MaterialFunctions_Utilities/MF_HeightBlendLandscapeLayers_OpacityMaskDisplaceCustom}
  FunctionOutput(None=LandscapeLayerSwitch!(LayerUsed=@n9.BlendedMaterial, LayerNotUsed=@n4) {parameterName=BiomeA}) {outputName=BlendedMaterial, sortPriority=-300}
  n10 = NamedRerouteUsage()
  FunctionOutput(None=QualitySwitch(Default=@n10, Medium=@n10)) {outputName=MobileBlendedMaterial, sortPriority=-100}
  FunctionOutput(None=ShadingPathSwitch(Default=NamedRerouteUsage(), Mobile=NamedRerouteUsage())) {outputName=LayerMask, sortPriority=-1}
  n11 = MaterialFunctionCall!(
   InputMaterial=@n10
   LayerMaterial=NamedRerouteUsage()
   LayerWeight=NamedRerouteUsage()
   UseContrast=NamedRerouteUsage()
   Contrast=NamedRerouteUsage()
  ) {materialFunction=/Game/Athena/Helios/Environments/Landscape/Materials/MaterialFunctions_Utilities/MF_HeightBlendLandscapeLayers_OpacityMask}
  LandscapeLayerSwitch(LayerUsed=@n11.BlendedMaterial, LayerNotUsed=@n10) {parameterName=BiomeA}
  n12 = Multiply(A=LandscapeLayerCoords() {mappingScale=10}, B=ScalarParameter() {defaultValue=2, group=ForestFloor, parameterName=DoomstadtLeaves_Scale})
  n13 = TextureSample(UVs=@n12) {texture=/Game/Environments/Helios/Terrain/Textures/T_Helios_Leaves_01_D}
  n14 = VectorParameter() {
   group=ForestFloor
   parameterName=DoomstadtLeaves_Tint
   channelNames {a=Desaturation, b="", g="", r=""}
   defaultValue=<rgba 1 0.0786 0.1682 1>
  }
  n15 = Desaturation(None=@n13.RGB, Fraction=@n14.Desaturation)
  LinearInterpolate(A=@n13.RGB, B=Saturate(None=Multiply(A=@n14, B=@n15)), Alpha=@n13.RGB)
  MaterialFunctionCall(InputMaterial=@n4, LayerMaterial=@n5, LayerWeight=@n6, UseContrast=@n7, Contrast=@n8) {materialFunction=&MF_HeightBlendLandscapeLayers_OpacityMask}
  Multiply(A=LandscapeLayerCoords() {mappingScale=10}, B=Constant() {r=1.5})
  n16 = FunctionInput(Preview=StaticBool()) {
   bUsePreviewValueAsDefault=true
   description="Default = False"
   inputName=UseTangentNormals
   inputType=FunctionInput_StaticBool
   sortPriority=15
  }
  NamedRerouteDeclaration(None=@n16) {name=UseTangentNormalsBool, nodeColor=<rgba 0.5 0 1 1>}
  NamedRerouteDeclaration(None=FunctionInput(Preview=Constant()) {bUsePreviewValueAsDefault=true, inputName=VariationMasks Mobile, inputType=FunctionInput_Vector4, sortPriority=36}) {name=VariationMasks Mobile, nodeColor=<rgba 1 0.9131 0 1>}
  n17 = SetMaterialAttributes!(MaterialAttributes)
  NamedRerouteDeclaration(None=FunctionInput(Preview=@n17) {bUsePreviewValueAsDefault=true, inputName=Material, inputType=FunctionInput_MaterialAttributes}) {name=InputMaterial, nodeColor=<rgba 0 1 0 1>}
  n18 = MaterialFunctionCall!(Scale=ScalarParameter() {defaultValue=900, group=ForestFloor, parameterName=ForestFloor_Scale}) {materialFunction=/Game/Athena/Apollo/Environments/Landscape/Materials/Biome/CommonMaterials/MF_SimpleWorldUV}
  n19 = TextureSampleParameter2D(UVs=@n18.Result) {
   group=ForestFloor
   parameterName=ForestFloor_D
   samplerSource=SSM_Wrap_WorldGroupSettings
   texture=/Game/Environments/Helios/Terrain/Textures/T_Helios_Bor_ForestFloor_02_D
  }
  n20 = TextureSampleParameter2D(UVs=@n18.Result) {
   group=ForestFloor
   parameterName=ForestFloor_S
   samplerSource=SSM_Wrap_WorldGroupSettings
   samplerType=SAMPLERTYPE_Masks
   texture=/Game/Environments/Helios/Terrain/Textures/T_Helios_Bor_ForestFloor_02_S
  }
  n21 = TextureSampleParameter2D(UVs=@n18.Result) {
   group=ForestFloor
   parameterName=ForestFloor_N
   samplerSource=SSM_Wrap_WorldGroupSettings
   samplerType=SAMPLERTYPE_Normal
   texture=/Game/Environments/Helios/Terrain/Textures/T_Helios_Bor_ForestFloor_02_N
  }
  n22 = SetMaterialAttributes!(
   MaterialAttributes
   BaseColor=@n19.RGB
   Specular=@n20.R
   Roughness=@n20.R
   Normal=MaterialFunctionCall!(
    Normals=@n21.RGB
    TransformToTangentSpace=NamedRerouteUsage()
    Normal Strength=ScalarParameter() {defaultValue=1, group=ForestFloor, parameterName=Grass_BorealForest_NormalAmt}
   ) {materialFunction=/Game/Athena/Apollo/Environments/Landscape/Materials/Biome/CommonMaterials/MF_HQ_WorldAlignedNormal}.Result
   OpacityMask=@n19.RGB
   Displacement=LinearInterpolate(B=@n19.A, Alpha=ScalarParameter() {defaultValue=0.75, group=Displacement, parameterName=ForestFloor_DisplacementAmt}) {constA=0.5}
  ) {attributeSetTypes [
    #1
    #2
    #3
    #4
    #5
    #6
   ]}
  n23 = TextureSample(UVs=@n12) {samplerType=SAMPLERTYPE_Masks, texture=/Game/Environments/Helios/Terrain/Textures/T_Helios_Leaves_01_S}
  n24 = SetMaterialAttributes(
   BaseColor=@n13.RGB
   Specular=@n23.R
   Roughness=@n23.R
   Normal=TextureSample(UVs=@n12) {samplerType=SAMPLERTYPE_Normal, texture=/Game/Environments/Helios/Terrain/Textures/T_Helios_Leaves_01_N}.RGB
   OpacityMask=@n13.RGB
   Displacement=@n13.RGB
  ) {attributeSetTypes [
    #1
    #2
    #3
    #4
    #5
    #6
   ]}
  n25 = Saturate(None=Multiply(A=NamedRerouteUsage(), B=ScalarParameter() {defaultValue=1.25, group=ForestFloor, parameterName=DoomstadtLeaves_WorldMaskMod}))
  n26 = MaterialFunctionCall!(
   Diffuse 1
   Diffuse 2
   Transistion Phase=Max(A=@n19.A, B=OneMinus(None=@n25))
   Height Texture 1=@n13.A
   Height Texture 2=@n19.A
   Contrast=ScalarParameter() {defaultValue=1, group=ForestFloor, parameterName=DoomstadtLeaves_BlendContrast}
  ) {materialFunction=/Engine/Functions/Engine_MaterialFunctions02/Texturing/HeightLerpWithTwoHeightMaps}
  n27 = MaterialFunctionCall!(
   Material=StaticSwitchParameter(True=BlendMaterialAttributes(A=@n22, B=@n24, Alpha=@n26.Alpha), False=@n22) {group=ForestFloor, parameterName=UseDoomstadtLeaves, sortPriority=-500}
   TerrainMasks_QuadA=NamedRerouteUsage()
   TerrainMasks_QuadB=NamedRerouteUsage()
   TerrainMasks_QuadC=NamedRerouteUsage()
   TerrainMasks_QuadD=NamedRerouteUsage()
  ) {materialFunction=/Game/Athena/Helios/Environments/Landscape/Materials/MaterialFunctions_Utilities/MF_ForestFloor_BiomeColorTint}
  MaterialExpressionReroute_0 = Reroute(None=@n27.Result)
  NamedRerouteDeclaration(None=@MaterialExpressionReroute_0.Output) {name=LayerMaterial, nodeColor=<rgba 0 1 0 1>}
  NamedRerouteDeclaration(None=FunctionInput(Preview=Constant()) {bUsePreviewValueAsDefault=true, inputName=LayerWeight, inputType=FunctionInput_Scalar, sortPriority=15}) {name=LayerWeight, nodeColor=<rgba 0.5 0 1 1>}
  NamedRerouteDeclaration(None=@n9.LayerMask) {name=LayerMask, nodeColor=<rgba 0 0.2734 1 1>}
  NamedRerouteDeclaration(None=@n11.LayerMaskNoContrast) {name=MobileLayerMask, nodeColor=<rgba 0 1 0.1094 1>}
  MaterialExpressionReroute_2 = Reroute(None=@n2)
  NamedRerouteDeclaration(None=@MaterialExpressionReroute_2.DoomstadtTrees_Mask) {name=BiomeMask, nodeColor=<rgba 0.5 0 1 1>}
  NamedRerouteDeclaration(None=FunctionInput(Preview=Constant()) {bUsePreviewValueAsDefault=true, inputName=TerrainMasks_QuadA, inputType=FunctionInput_Vector4, sortPriority=20}) {name=TerrainMasks_QuadA, nodeColor=<rgba 0.5 0 1 1>}
  MaterialExpressionReroute_3 = Reroute(None=@n3)
  NamedRerouteDeclaration(None=@MaterialExpressionReroute_3.LayerWeight) {name=LayerWeight1, nodeColor=<rgba 0.5 0 1 1>}
  NamedRerouteDeclaration(None=FunctionInput(Preview=Constant()) {bUsePreviewValueAsDefault=true, inputName=TerrainMasks_QuadD, inputType=FunctionInput_Vector4, sortPriority=35}) {name=TerrainMasks_QuadD, nodeColor=<rgba 0.5 0 1 1>}
  NamedRerouteDeclaration(None=FunctionInput(Preview=Constant()) {bUsePreviewValueAsDefault=true, inputName=TerrainMasks_QuadB, inputType=FunctionInput_Vector4, sortPriority=25}) {name=TerrainMasks_QuadB, nodeColor=<rgba 0.5 0 1 1>}
  NamedRerouteDeclaration(None=FunctionInput(Preview=@n17) {bUsePreviewValueAsDefault=true, inputName=MobileMaterial, inputType=FunctionInput_MaterialAttributes, sortPriority=10}) {name=MobileInputMaterial, nodeColor=<rgba 0 0.5 1 1>}
  n28 = FunctionInput() {bUsePreviewValueAsDefault=true, inputName=MobileBorealForestMaterial, inputType=FunctionInput_MaterialAttributes, sortPriority=11}
  NamedRerouteDeclaration(None=MaterialFunctionCall!(MobileBorealForestMaterial=@n28) {materialFunction=/Game/Athena/Helios/Environments/Landscape/Materials/QualityLevelLayers/MF_Helios_Mobile_Forest}.Result) {name=MobileLayerMaterial, nodeColor=<rgba 0 0.5 1 1>}
  NamedRerouteDeclaration(None=StaticBool()) {name=UseContrast, nodeColor=<rgba 0.5 0 1 1>}
  NamedRerouteDeclaration(None=ScalarParameter() {group=ForestFloor, parameterName=ForestFloor_LayerBlendContrast}) {name=LayerBlendContrast, nodeColor=<rgba 0.5 0 1 1>}
  PinBase() {reroutePins [
    DoomstadtTrees_Mask {expression=&self:MaterialExpressionReroute_2}
    LayerWeight {expression=&self:MaterialExpressionReroute_3}
   ]}
  PinBase!(Output=@n27.Result) {reroutePins [Output {expression=&self:MaterialExpressionReroute_0}]}
  Power(Base=NamedRerouteUsage(), Exp=ScalarParameter() {defaultValue=1.5, group=ForestFloor, parameterName=DoomstadtLeaves_LayerMaskPower})
  Saturate(None=Multiply(A=VectorParameter() {group=ForestFloor, parameterName=ForestFloor_ColorTint, defaultValue=<rgba 1 1 1 0>}.RGB, B=@n19.RGB))
  n29 = ScalarParameter() {defaultValue=1.5, group=Displacement, parameterName=ForestFloor_DisplacementPower}
  Saturate(None=Multiply(A=Power(Base=@n19.A, Exp=@n29), B=ScalarParameter() {defaultValue=2, group=Displacement, parameterName=ForestFloor_DisplacementMod}))
  Saturate(None=Multiply(A=ScalarParameter() {defaultValue=1, group=ForestFloor, parameterName=DoomstadtLeaves_FFTransMod}, B=@n19.A))
  n30 = ScalarParameter() {defaultValue=1, group=Displacement, parameterName=DoomstadtLeaves_DisplacementPower}
  Saturate(None=Multiply(A=Power(Base=@n13.A, Exp=@n30), B=ScalarParameter() {defaultValue=1, group=Displacement, parameterName=DoomstadtLeaves_DisplacementMod}))
 ]}
