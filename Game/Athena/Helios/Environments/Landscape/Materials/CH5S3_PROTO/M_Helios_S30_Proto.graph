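# Prototype Helios landscape material (CH5S3, S30) with wasteland and desert layers, supporting Nanite and virtual heightfield mesh.
# MF_Helios layer functions chain grass through asphalt using painted weights and terrain masks, with cliff and static-switched biome layers like forest floor, scree and snow.
# Parallel mobile materials accompany the high-end path.
# Layer and biome masks also drive landscape grass spawning and priority-ordered physical materials, with temporary Lobby Island and biome-mask overrides.
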Material /Game/Athena/Helios/Environments/Landscape/Materials/CH5S3_PROTO/M_Helios_S30_Proto
graph {
 properties {
  bTangentSpaceNormal=false
  bUseLightmapDirectionality=false
  bUseMaterialAttributes=true
  bUsedWithNanite=true
  bUsedWithVirtualHeightfieldMesh=true
  pixelDepthOffsetMode=PDOM_Legacy
  shadingModel=MSM_FromMaterialExpression
  usedShadingModels=Default Lit
  displacementScaling {center=0.5, magnitude=0.1}
 }
 nodes [
  Composite() {subgraphName=BRL_Flowers}
  Composite() {subgraphName=Stone_Physmat_Mask}
  Composite() {subgraphName=Dirt_Physmat_Mask}
  n1 = NamedRerouteUsage()
  n2 = NamedRerouteUsage()
  Composite!("HighEndMaterial(MA)"=@n1, "MobileMaterial(MA)"=@n2)
  Composite!()
  Composite() {subgraphName=CHPC_Layer01}
  Composite() {subgraphName=Shared_Grass_Layer01}
  Composite() {subgraphName=Snow_Physmat_Mask}
  Composite() {subgraphName=CHPG_Layer01}
  Composite()
  Composite() {subgraphName=Sand_Physmat_Mask}
  Composite() {subgraphName=Grass_Physmat_Mask}
  Constant() {r=-2}
  Constant() {r=3}
  Constant() {r=1}
  Constant() {r=-1}
  Constant() {r=-2}
  Constant() {r=1}
  Constant()
  Constant()
  GetMaterialAttributes() {attributeGetTypes [#1]}
  MaterialExpressionNamedRerouteUsage_171 = NamedRerouteUsage()
  MaterialExpressionNamedRerouteUsage_46 = NamedRerouteUsage()
  MaterialExpressionNamedRerouteUsage_64 = NamedRerouteUsage()
  MaterialExpressionNamedRerouteUsage_54 = NamedRerouteUsage()
  MaterialExpressionNamedRerouteUsage_35 = NamedRerouteUsage()
  MaterialExpressionNamedRerouteUsage_48 = NamedRerouteUsage()
  MaterialExpressionNamedRerouteUsage_147 = NamedRerouteUsage()
  MaterialExpressionNamedRerouteUsage_144 = NamedRerouteUsage()
  MaterialExpressionNamedRerouteUsage_156 = NamedRerouteUsage()
  MaterialExpressionNamedRerouteUsage_163 = NamedRerouteUsage()
  MaterialExpressionNamedRerouteUsage_247 = NamedRerouteUsage()
  LandscapeGrassOutput!(
   Lobby_Grass_Layer01=@MaterialExpressionNamedRerouteUsage_171
   Shared_Grass_Layer01=@MaterialExpressionNamedRerouteUsage_46
   CHPG_Layer01=@MaterialExpressionNamedRerouteUsage_64
   CHPC_Layer01=@MaterialExpressionNamedRerouteUsage_54
   BRLF_Layer01=@MaterialExpressionNamedRerouteUsage_35
   BRLF_Layer02=@MaterialExpressionNamedRerouteUsage_48
   BRL_Flowers=@MaterialExpressionNamedRerouteUsage_147
   BRL_Snow=@MaterialExpressionNamedRerouteUsage_144
   Shared_Scree=@MaterialExpressionNamedRerouteUsage_156
   Shared_Gravel_Pebbles01=@MaterialExpressionNamedRerouteUsage_163
   SunGold01=@MaterialExpressionNamedRerouteUsage_247
   Wasteland=@MaterialExpressionNamedRerouteUsage_247
  ) {
   desc="This may increase reported texture lookups,\r\nbut it *probably* gets compiled out on cook."
   grassTypes [
    Lobby_Grass_Layer01 {
     grassType=/Game/Athena/Helios/Environments/Landscape/LandscapeFoliageLayers/LS_GrassLayer_01_Lobby
     input {
      expression=&self:MaterialExpressionNamedRerouteUsage_171
      inputName=None
      mask=0
      maskA=0
      maskB=0
      maskG=0
      maskR=0
      outputIndex=0
     }
    }
    Shared_Grass_Layer01 {
     grassType=/Game/Athena/Helios/Environments/Landscape/LandscapeFoliageLayers/LS_GrassLayer_01
     input {
      expression=&self:MaterialExpressionNamedRerouteUsage_46
      inputName=None
      mask=0
      maskA=0
      maskB=0
      maskG=0
      maskR=0
      outputIndex=0
     }
    }
    CHPG_Layer01 {
     grassType=/Game/Athena/Helios/Environments/Landscape/LandscapeFoliageLayers/LS_ChapparalGrass_Layer
     input {
      expression=&self:MaterialExpressionNamedRerouteUsage_64
      inputName=None
      mask=0
      maskA=0
      maskB=0
      maskG=0
      maskR=0
      outputIndex=0
     }
    }
    CHPC_Layer01 {
     grassType=/Game/Athena/Helios/Environments/Landscape/LandscapeFoliageLayers/LS_ScotchBroomLayer_01
     input {
      expression=&self:MaterialExpressionNamedRerouteUsage_54
      inputName=None
      mask=0
      maskA=0
      maskB=0
      maskG=0
      maskR=0
      outputIndex=0
     }
    }
    BRLF_Layer01 {
     grassType=/Game/Athena/Helios/Environments/Landscape/LandscapeFoliageLayers/LS_ForestLayer_01
     input {
      expression=&self:MaterialExpressionNamedRerouteUsage_35
      inputName=None
      mask=0
      maskA=0
      maskB=0
      maskG=0
      maskR=0
      outputIndex=0
     }
    }
    BRLF_Layer02 {
     grassType=/Game/Athena/Helios/Environments/Landscape/LandscapeFoliageLayers/LS_ForestLayer_03
     input {
      expression=&self:MaterialExpressionNamedRerouteUsage_48
      inputName=None
      mask=0
      maskA=0
      maskB=0
      maskG=0
      maskR=0
      outputIndex=0
     }
    }
    BRL_Flowers {
     grassType=/Game/Athena/Helios/Environments/Landscape/LandscapeFoliageLayers/LS_TundraFlowers
     input {
      expression=&self:MaterialExpressionNamedRerouteUsage_147
      inputName=None
      mask=0
      maskA=0
      maskB=0
      maskG=0
      maskR=0
      outputIndex=0
     }
    }
    BRL_Snow {
     grassType=/Game/Athena/Helios/Environments/Landscape/LandscapeFoliageLayers/LS_SnowLayer_01
     input {
      expression=&self:MaterialExpressionNamedRerouteUsage_144
      inputName=None
      mask=0
      maskA=0
      maskB=0
      maskG=0
      maskR=0
      outputIndex=0
     }
    }
    Shared_Scree {
     grassType=/Game/Athena/Helios/Environments/Landscape/LandscapeFoliageLayers/LS_Scree_01
     input {
      expression=&self:MaterialExpressionNamedRerouteUsage_156
      inputName=None
      mask=0
      maskA=0
      maskB=0
      maskG=0
      maskR=0
      outputIndex=0
     }
    }
    Shared_Gravel_Pebbles01 {
     grassType=/Game/Athena/Helios/Environments/Landscape/LandscapeFoliageLayers/LS_Gravel_Pebbles_01
     input {
      expression=&self:MaterialExpressionNamedRerouteUsage_163
      inputName=None
      mask=0
      maskA=0
      maskB=0
      maskG=0
      maskR=0
      outputIndex=0
     }
    }
    SunGold01 {
     grassType=/Game/Athena/Helios/Environments/Landscape/Materials/CH5S3_PROTO/GrassLayer/LS_WasteLandShrubs
     input {
      expression=&self:MaterialExpressionNamedRerouteUsage_247
      inputName=None
      mask=0
      maskA=0
      maskB=0
      maskG=0
      maskR=0
      outputIndex=0
     }
    }
    Wasteland {
     grassType=&LS_WasteLandShrubs
     input {
      expression=&self:MaterialExpressionNamedRerouteUsage_247
      inputName=None
      mask=0
      maskA=0
      maskB=0
      maskG=0
      maskR=0
      outputIndex=0
     }
    }
   ]
  }
  MaterialExpressionNamedRerouteUsage_77 = NamedRerouteUsage()
  MaterialExpressionNamedRerouteUsage_82 = NamedRerouteUsage()
  MaterialExpressionNamedRerouteUsage_88 = NamedRerouteUsage()
  MaterialExpressionNamedRerouteUsage_169 = NamedRerouteUsage()
  MaterialExpressionNamedRerouteUsage_87 = NamedRerouteUsage()
  MaterialExpressionNamedRerouteUsage_89 = NamedRerouteUsage()
  LandscapePhysicalMaterialOutput!(
   Dirt=@MaterialExpressionNamedRerouteUsage_77
   Stone=@MaterialExpressionNamedRerouteUsage_82
   Grass=@MaterialExpressionNamedRerouteUsage_88
   Asphalt=@MaterialExpressionNamedRerouteUsage_169
   Sand=@MaterialExpressionNamedRerouteUsage_87
   Snow=@MaterialExpressionNamedRerouteUsage_89
  ) {
   desc="These are layered in order of importance.\r\nIf Dirt=1 and Stone=1, then Dirt will \"win\".\r\nTheoretically, these values aren't clamped.\r\nSo a value of 2 should trump a value of 1."
   inputs [
    {
     physicalMaterial=/Game/FortressPhysicalMaterials/Dirt
     input {
      expression=&self:MaterialExpressionNamedRerouteUsage_77
      inputName=None
      mask=0
      maskA=0
      maskB=0
      maskG=0
      maskR=0
      outputIndex=0
     }
    }
    {
     physicalMaterial=/Game/FortressPhysicalMaterials/Stone
     input {
      expression=&self:MaterialExpressionNamedRerouteUsage_82
      inputName=None
      mask=0
      maskA=0
      maskB=0
      maskG=0
      maskR=0
      outputIndex=0
     }
    }
    {
     physicalMaterial=/Game/FortressPhysicalMaterials/Grass
     input {
      expression=&self:MaterialExpressionNamedRerouteUsage_88
      inputName=None
      mask=0
      maskA=0
      maskB=0
      maskG=0
      maskR=0
      outputIndex=0
     }
    }
    {
     physicalMaterial=/Game/FortressPhysicalMaterials/Asphalt
     input {
      expression=&self:MaterialExpressionNamedRerouteUsage_169
      inputName=None
      mask=0
      maskA=0
      maskB=0
      maskG=0
      maskR=0
      outputIndex=0
     }
    }
    {
     physicalMaterial=/Game/FortressPhysicalMaterials/Sand
     input {
      expression=&self:MaterialExpressionNamedRerouteUsage_87
      inputName=None
      mask=0
      maskA=0
      maskB=0
      maskG=0
      maskR=0
      outputIndex=0
     }
    }
    {
     physicalMaterial=/Game/FortressPhysicalMaterials/Snow
     input {
      expression=&self:MaterialExpressionNamedRerouteUsage_89
      inputName=None
      mask=0
      maskA=0
      maskB=0
      maskG=0
      maskR=0
      outputIndex=0
     }
    }
   ]
  }
  n3 = NamedRerouteUsage()
  n4 = LinearInterpolate(
   A=NamedRerouteUsage()
   B=LinearInterpolate(A=NamedRerouteUsage(), B=NamedRerouteUsage(), Alpha=ScalarParameter() {group=Cliff, parameterName=SnowCliff_DEBUGMe6, sortPriority=98})
   Alpha=ScalarParameter() {defaultValue=1, group=Cliff, parameterName=SnowCliff_DEBUGMe5, sortPriority=98}
  )
  n5 = ScalarParameter() {defaultValue=1, group=Cliff, parameterName=SnowCliff_DEBUGMe4, sortPriority=98}
  LinearInterpolate(
   A=NamedRerouteUsage()
   B=LinearInterpolate(
    A=NamedRerouteUsage()
    B=LinearInterpolate(A=@n3, B=@n4, Alpha=@n5)
    Alpha=ScalarParameter() {defaultValue=1, group=Cliff, parameterName=SnowCliff_DEBUGMe3, sortPriority=98}
   )
   Alpha=ScalarParameter() {defaultValue=1, group=Cliff, parameterName=SnowCliff_DEBUGMe2, sortPriority=98}
  )
  n6 = NamedRerouteUsage()
  n7 = NamedRerouteUsage()
  n8 = NamedRerouteUsage()
  n9 = OneMinus(None=Max(A=Max(A=@n7, B=@n8), B=Constant() {desc=Replace With SnowLayer_Alpha}))
  Max(A=@n6, B=@n9)
  n10 = Max(A=NamedRerouteUsage(), B=NamedRerouteUsage())
  Max(A=NamedRerouteUsage(), B=@n10)
  Multiply(A=NamedRerouteUsage(), B=@n9)
  n11 = OneMinus(None=NamedRerouteUsage())
  Multiply(A=@n11, B=NamedRerouteUsage())
  n12 = StaticBoolParameter() {
   desc="Temp workaround for LobbyIsland so at least the layers can\r\nbe painted until the RVT is setup correctly"
   group=Temp
   parameterName=IsLobbyIsland
  }
  NamedRerouteDeclaration(None=@n12) {name=IsLobbyIsland, nodeColor=<rgba 0.5 0 1 1>}
  n13 = MaterialFunctionCall!(
   Material
   MobileMaterial
   IsForStaticMesh
   UseTangentNormals
   UseCustomSlopeMask
   CustomSlopeMask
   Grass_EdgeOnly=NamedRerouteUsage()
   LayerWeight=NamedRerouteUsage()
   TerrainMasks_QuadA=NamedRerouteUsage()
   TerrainMasks_QuadB=NamedRerouteUsage()
   TerrainMasks_QuadC=NamedRerouteUsage()
   TerrainMasks_QuadD=NamedRerouteUsage()
   VariationMasks Mobile=NamedRerouteUsage()
  ) {materialFunction=/Game/Athena/Helios/Environments/Landscape/Materials/LandscapeLayers/MF_Helios_Grass}
  NamedRerouteDeclaration(None=@n13.LayerMask) {name=GrassLayerMask, nodeColor=<rgba 0.5 0 1 1>}
  n14 = MaterialFunctionCall!(
   Material=NamedRerouteDeclaration(None=@n13.BlendedMaterial) {name=Layer0_DefaultGrass, nodeColor=<rgba 0.5 0 1 1>}
   MobileMaterial=NamedRerouteDeclaration(None=@n13.MobileBlendedMaterial) {name=Layer0_DefaultGrassMobile, nodeColor=<rgba 0.5 0 1 1>}
   UseTangentNormals
   LayerWeight=NamedRerouteUsage()
   TerrainMasks_QuadA=NamedRerouteUsage()
   TerrainMasks_QuadB=NamedRerouteUsage()
   TerrainMasks_QuadC=NamedRerouteUsage()
   TerrainMasks_QuadD=NamedRerouteUsage()
   VariationMasks Mobile=NamedRerouteUsage()
  ) {materialFunction=&/MF_Helios_Wasteland}
  n15 = MaterialFunctionCall!(
   Material=@n14.BlendedMaterial
   MobileMaterial=@n14.BlendedMaterial
   IsForStaticMesh
   UseTangentNormals
   LayerWeight=NamedRerouteUsage()
   MudWeight=NamedRerouteUsage()
   TerrainMasks_QuadA=NamedRerouteUsage()
   TerrainMasks_QuadB=NamedRerouteUsage()
   TerrainMasks_QuadC=NamedRerouteUsage()
   TerrainMasks_QuadD=NamedRerouteUsage()
   VariationMasks Mobile=NamedRerouteUsage()
  ) {materialFunction=/Game/Athena/Helios/Environments/Landscape/Materials/LandscapeLayers/MF_Helios_Gravel}
  n16 = NamedRerouteDeclaration(None=@n15.LayerMask) {name=GravelLayerMask, nodeColor=<rgba 0.5 0 1 1>}
  n17 = MaterialFunctionCall!(
   Material=MaterialFunctionCall!(
    LayerWEight=@n16
    Color=VectorParameter() {parameterName=GravelColorTInt, defaultValue=<rgba 0.5312 0.4727 0.4372 1>}.RGB
    DesatFactor=ScalarParameter() {defaultValue=0.5, parameterName=GravelDest}
    Value=ScalarParameter() {defaultValue=1.5, parameterName=GravelVal}
    Material=@n15.BlendedMaterial
   ) {materialFunction=&/MF_WastelandColorTint}.Material
   MobileMaterial=@n15.MobileBlendedMaterial
   MobileMudMaterial=@n15.MobileBlendedMaterial
   ForMudMaterial=@n15.MobileBlendedMaterial
   LayerWeight=NamedRerouteUsage()
   UseTangentNormals
   TerrainMasks_QuadA=NamedRerouteUsage()
   TerrainMasks_QuadB=NamedRerouteUsage()
   TerrainMasks_QuadC=NamedRerouteUsage()
   TerrainMasks_QuadD=NamedRerouteUsage()
   VariationMasks Mobile=NamedRerouteUsage()
  ) {materialFunction=/Game/Athena/Helios/Environments/Landscape/Materials/LandscapeLayers/MF_Helios_Mud}
  n18 = MaterialFunctionCall!(
   Material=@n17.BlendedMaterial
   MobileMaterial=@n17.MobileBlendedMaterial
   LayerWeight=NamedRerouteUsage()
   UseTangentNormals
   TerrainMasks_QuadA=NamedRerouteUsage()
   TerrainMasks_QuadB=NamedRerouteUsage()
   TerrainMasks_QuadC=NamedRerouteUsage()
   TerrainMasks_QuadD=NamedRerouteUsage()
   VariationMasks Mobile=NamedRerouteUsage()
   OD_MD_Normals
   OD_MD_SedimentLayers
  ) {materialFunction=/Game/Athena/Helios/Environments/Landscape/Materials/LandscapeLayers/MF_Helios_Sand}
  NamedRerouteDeclaration(None=@n18.LayerMask) {name=SandLayerMask, nodeColor=<rgba 0.5 0 1 1>}
  NamedRerouteDeclaration(None=@n17.LayerMask) {name=MudLayerMask, nodeColor=<rgba 0.5 0 1 1>}
  n19 = MaterialFunctionCall!(
   Material=@n18.BlendedMaterial
   MobileMaterial=@n18.BlendedMaterial
   UseTangentNormals
   LayerWeight=NamedRerouteUsage()
   TerrainMasks_QuadA=NamedRerouteUsage()
   TerrainMasks_QuadB=NamedRerouteUsage()
   TerrainMasks_QuadC=NamedRerouteUsage()
   TerrainMasks_QuadD=NamedRerouteUsage()
   VariationMasks Mobile=NamedRerouteUsage()
  ) {materialFunction=&/MF_Helios_Desert}
  n20 = MaterialFunctionCall!(
   Material=@n19.BlendedMaterial
   MobileMaterial=@n19.BlendedMaterial
   UseTangentNormals
   LayerWeight=Constant()
   TerrainMasks_QuadA=NamedRerouteUsage()
   TerrainMasks_QuadB=NamedRerouteUsage()
   TerrainMasks_QuadC=NamedRerouteUsage()
   TerrainMasks_QuadD=NamedRerouteUsage()
   VariationMasks Mobile=NamedRerouteUsage()
  ) {materialFunction=/Game/Athena/Helios/Environments/Landscape/Materials/LandscapeLayers/MF_Helios_Asphalt}
  NamedRerouteDeclaration(None=@n20.LayerMask) {name=AsphaltLayerMask, nodeColor=<rgba 0.5 0 1 1>}
  n21 = MaterialFunctionCall!(
   Material=NamedRerouteUsage()
   MobileMaterial=NamedRerouteUsage()
   LayerWeight=NamedRerouteUsage()
   UseTangentNormals
   TerrainMasks_QuadA=NamedRerouteUsage()
   TerrainMasks_QuadB=NamedRerouteUsage()
   TerrainMasks_QuadC=NamedRerouteUsage()
   TerrainMasks_QuadD=NamedRerouteUsage()
   VariationMasks Mobile=NamedRerouteUsage()
   MossDesaturateMask=NamedRerouteUsage()
  ) {materialFunction=/Game/Athena/Helios/Environments/Landscape/Materials/LandscapeLayers/MF_Helios_Cliff}
  NamedRerouteDeclaration(None=@n21.GrassMask) {name=CliffLayerGrassMask, nodeColor=<rgba 0.6172 0 1 1>}
  n22 = MaterialFunctionCall!(
   Material=NamedRerouteUsage()
   MobileMaterial=NamedRerouteUsage()
   IsForStaticMesh
   UseTangentNormals
   LayerWeight=NamedRerouteUsage()
   TerrainMasks_QuadA
   TerrainMasks_QuadB
   TerrainMasks_QuadC
   TerrainMasks_QuadD
   CliffMaterialOnly=NamedRerouteUsage()
   CliffLayerMask=NamedRerouteUsage()
   VariationMasks Mobile=NamedRerouteUsage()
  ) {materialFunction=/Game/Athena/Helios/Environments/Landscape/Materials/LandscapeLayers/MF_Helios_Snow_LSLayer}
  n23 = NamedRerouteDeclaration(None=StaticBoolParameter() {group=BiomeABCD_Settings, parameterName=UseSnowForBiomeC}) {name=UseSnowForBiomeC, nodeColor=<rgba 0.5 0 1 1>}
  NamedRerouteDeclaration(None=StaticSwitch(True=@n22.BlendedMaterial, False=NamedRerouteUsage(), Value=@n23)) {name=Highend:SurfaceMaterialResult, nodeColor=<rgba 0.5 0 1 1>}
  n24 = MaterialFunctionCall!(UseLandscapeCoords, "UseProjection(translucent)") {materialFunction=/Game/Athena/Helios/Environments/Landscape/Materials/MaterialFunctions_Utilities/MF_TerrainMasks}
  NamedRerouteDeclaration(None=@n24.QuadA_RGBA) {name=TerrainMasks_QuadA, nodeColor=<rgba 0.5 0 1 1>}
  NamedRerouteDeclaration(None=@n24.QuadC_RGBA) {name=TerrainMasks_QuadC, nodeColor=<rgba 0.5 0 1 1>}
  NamedRerouteDeclaration(None=@n21.LayerMask) {name=CliffLayerMask, nodeColor=<rgba 0.6172 0 1 1>}
  n25 = Constant()
  n26 = MaterialFunctionCall(X=Constant() {r=1}, Y=@n25, Z=@n25, A=@n25) {materialFunction=&MakeFloat4}
  n27 = NamedRerouteUsage()
  n28 = MaterialFunctionCall(Float4=NamedRerouteDeclaration(None=StaticSwitch(True=@n26.Result, False=@n24.QuadB_RGBA, Value=@n27)) {name=TerrainMasks_QuadB, nodeColor=<rgba 0.5 0 1 1>}) {materialFunction=&BreakOutFloat4Components}
  n29 = Max(A=@n28.R, B=@n28.R)
  n30 = Max(A=@n28.B, B=@n28.B)
  NamedRerouteDeclaration(None=OneMinus(None=Max(A=@n29, B=@n30))) {name=Biome_Chaparral_Grassland, desc=Acronym:CHPG, nodeColor=<rgba 0.5 0 1 1>}
  NamedRerouteDeclaration(None=@n22.GrassMask) {name=SnowLayerGrassMask, nodeColor=<rgba 0.6172 0 1 1>}
  n31 = MaterialFunctionCall!(
   Material=NamedRerouteUsage()
   MobileMaterial=NamedRerouteUsage()
   GrassWeight
   UseTangentNormals
   LayerWeight=NamedRerouteUsage()
   TerrainMasks_QuadA=NamedRerouteUsage()
   TerrainMasks_QuadB=NamedRerouteUsage()
   TerrainMasks_QuadC=NamedRerouteUsage()
   TerrainMasks_QuadD=NamedRerouteUsage()
   VariationMasks Mobile=NamedRerouteUsage()
   Chap_D
   Chap_S
   Chap_N
  ) {materialFunction=/Game/Athena/Helios/Environments/Landscape/Materials/LandscapeLayers/MF_Helios_Grass_ChaparralCoast}
  n32 = Multiply(A=@n31.ChaparralCoast_GrassMask, B=ScalarParameter() {defaultValue=2, group=GrassMeshes, parameterName=ChaparralCoast_GrassMeshMask_Pow})
  n33 = ScalarParameter() {defaultValue=-0.5, group=GrassMeshes, parameterName=ChaparralCoast_GrassMeshMask_Add}
  NamedRerouteDeclaration(None=Saturate(None=Add(A=@n32, B=@n33))) {name=ChaparralCoast_GrassMask, nodeColor=<rgba 0.5 0 1 1>}
  n34 = NamedRerouteDeclaration(None=LandscapeLayerSample() {parameterName=Gravel}) {name=Gravel_Weight, nodeColor=<rgba 0.5 0 1 1>}
  n35 = NamedRerouteDeclaration(None=LandscapeLayerSample() {parameterName=Mud}) {name=Mud_Weight, nodeColor=<rgba 0.5 0 1 1>}
  n36 = Max(
   A=NamedRerouteDeclaration(None=LandscapeLayerSample() {parameterName=Sand}) {name=Sand_Weight, nodeColor=<rgba 0.5 0 1 1>}
   B=NamedRerouteDeclaration(None=LandscapeLayerSample() {parameterName=Asphalt}) {name=Asphalt_Weight, nodeColor=<rgba 0.5 0 1 1>}
  )
  n37 = NamedRerouteDeclaration(None=LandscapeLayerSample() {parameterName=BiomeB}) {name=BiomeB_Weight, nodeColor=<rgba 0.5 0 1 1>}
  n38 = Max(A=Max(A=Max(A=@n34, B=@n35), B=Max(A=@n36, B=@n37)), B=Max(A=NamedRerouteUsage(), B=NamedRerouteUsage()))
  NamedRerouteDeclaration(None=OneMinus(None=@n38)) {name=Grass_Weight_NoCliff, nodeColor=<rgba 0.5 0 1 1>}
  NamedRerouteDeclaration(None=StaticSwitch(True=@n22.MobileBlendedMaterial, False=NamedRerouteUsage(), Value=@n23)) {name=MOBILE:SurfaceMaterialResult, nodeColor=<rgba 0.5 0 1 1>}
  NamedRerouteDeclaration(None=MaterialFunctionCall() {materialFunction=/Game/Athena/Helios/Environments/Landscape/Materials/MaterialFunctions_Utilities/MF_VariationMasks_Mobile}.VariationMasks Mobile) {name=VariationMasks Mobile, nodeColor=<rgba 1 0.9141 0 1>}
  NamedRerouteDeclaration(None=SetMaterialAttributes!(MaterialAttributes=GetMaterialAttributes(None=@n22.PostRVT).MaterialAttributes, Specular=NamedRerouteUsage()) {attributeSetTypes [#2]}) {name=Snow_PostRVT, nodeColor=<rgba 0.6172 0 1 1>}
  n39 = MaterialFunctionCall!(
   Material=NamedRerouteUsage()
   MobileMaterial=NamedRerouteUsage()
   UseTangentNormals
   LayerWeight=NamedRerouteUsage()
   TerrainMasks_QuadA=NamedRerouteUsage()
   TerrainMasks_QuadB=NamedRerouteUsage()
   TerrainMasks_QuadC=NamedRerouteUsage()
   TerrainMasks_QuadD=NamedRerouteUsage()
   VariationMasks Mobile=NamedRerouteUsage()
  ) {materialFunction=/Game/Athena/Helios/Environments/Landscape/Materials/LandscapeLayers/MF_Helios_Scree}
  n40 = NamedRerouteDeclaration(None=StaticBoolParameter() {defaultValue=true, group=BiomeABCD_Settings, parameterName=UseBiomeB_Scree, sortPriority=-500}) {name=UseBiomeB_Scree, nodeColor=<rgba 0.5 0 1 1>}
  NamedRerouteDeclaration(None=StaticSwitch(True=@n39.LayerMask, False=Constant(), Value=@n40)) {name=ScreeLayerMask, nodeColor=<rgba 0.5 0 1 1>}
  n41 = LinearInterpolate(
   A=ScalarParameter() {defaultValue=-2, group=GrassMeshes, parameterName=BRLF_Forest_Layer_01_SpawnLA}
   B=ScalarParameter() {defaultValue=1, group=GrassMeshes, parameterName=BRLF_Forest_Layer_01_SpawnLB}
   Alpha=Multiply(A=NamedRerouteUsage(), B=NamedRerouteUsage())
  )
  n42 = Multiply(A=Multiply(A=Saturate(None=@n41), B=@n11), B=NamedRerouteUsage())
  MaterialExpressionReroute_1 = Reroute(None=@n42)
  n43 = MaterialFunctionCall!(In=@MaterialExpressionReroute_1.Output) {materialFunction=/Game/Athena/Apollo/Environments/Landscape/Materials/Biome/CommonMaterials/MF_LS_GrassSwitch}
  NamedRerouteDeclaration(None=@n43.Result) {name=BRLF_Layer01, nodeColor=<rgba 0.5 0 1 1>}
  n44 = WorldPosition()
  n45 = LinearInterpolate(
   A=ScalarParameter() {defaultValue=-1, group=GrassMeshes, parameterName=BRLF_Forest_Noise_LA}
   B=ScalarParameter() {defaultValue=3, group=GrassMeshes, parameterName=BRLF_Forest_Noise_LB}
   Alpha=MaterialFunctionCall!(
    NoiseTexture
    Position=Add(A=@n44.XYZ) {constB=1024}
    MicroScale_01=ScalarParameter() {defaultValue=5000, group=GrassMeshes, parameterName=BRLF_Forest}
    MicroScale_02
    MicroScale_03
    MacroScale
   ) {materialFunction=/Game/Athena/Helios/Environments/Landscape/Materials/MaterialFunctions_Utilities/MF_VariationMasks}.MicroVariationMask_01
  )
  n46 = NamedRerouteUsage()
  n47 = ComponentMask(None=@n46) {g=true}
  n48 = ComponentMask(None=@n46) {b=true}
  n49 = Subtract(
   A=Multiply(A=Multiply(A=@n47, B=@n48), B=ScalarParameter() {defaultValue=20, parameterName=Contrast1})
   B=ScalarParameter() {defaultValue=3, parameterName=Contrast2}
  )
  NamedRerouteDeclaration(None=ShadingPathSwitch(Default=Saturate(None=@n45), Mobile=Saturate(None=@n49))) {name=BRLF_Forest_Variation_Noise, nodeColor=<rgba 0 0.0391 1 1>}
  n50 = MaterialFunctionCall!(
   Material=NamedRerouteUsage()
   MobileMaterial=NamedRerouteUsage()
   MobileBorealForestMaterial=NamedRerouteUsage()
   UseTangentNormals
   LayerWeight=NamedRerouteUsage()
   TerrainMasks_QuadA=NamedRerouteUsage()
   TerrainMasks_QuadB=NamedRerouteUsage()
   TerrainMasks_QuadC=NamedRerouteUsage()
   TerrainMasks_QuadD=NamedRerouteUsage()
   VariationMasks Mobile=NamedRerouteUsage()
  ) {materialFunction=/Game/Athena/Helios/Environments/Landscape/Materials/LandscapeLayers/MF_Helios_ForestFloor}
  n51 = NamedRerouteDeclaration(None=StaticBoolParameter() {defaultValue=true, group=BiomeABCD_Settings, parameterName=UseBiomeA_ForestFloor, sortPriority=-500}) {name=UseBiomeA_ForestFloor, nodeColor=<rgba 0.5 0 1 1>}
  NamedRerouteDeclaration(None=StaticSwitch(True=@n50.LayerMask, False=Constant(), Value=@n51)) {name=ForestFloorLayerMask, nodeColor=<rgba 0.5 0 1 1>}
  n52 = NamedRerouteUsage()
  n53 = Constant() {r=3}
  NamedRerouteDeclaration(None=Saturate(None=Multiply(A=@n52, B=@n53))) {name=Asphalt_physmat_Weightmap, nodeColor=<rgba 0.5 0 1 1>}
  n54 = LinearInterpolate(
   A=ScalarParameter() {defaultValue=-2, group=GrassMeshes, parameterName=BRLF_Forest_Layer_01_SpawnLA}
   B=ScalarParameter() {defaultValue=1, group=GrassMeshes, parameterName=BRLF_Forest_Layer_01_SpawnLB}
   Alpha=Multiply(A=NamedRerouteUsage(), B=NamedRerouteUsage())
  )
  n55 = OneMinus(None=NamedRerouteUsage())
  n56 = NamedRerouteUsage()
  n57 = Multiply(A=Multiply(A=Saturate(None=@n54), B=Multiply(A=@n55, B=@n56)), B=NamedRerouteUsage())
  MaterialExpressionReroute_4 = Reroute(None=@n57)
  n58 = MaterialFunctionCall(In=@MaterialExpressionReroute_4.Output) {materialFunction=&MF_LS_GrassSwitch}
  NamedRerouteDeclaration(None=@n58.Result) {name=BRLF_Layer02, nodeColor=<rgba 0.5 0 1 1>}
  NamedRerouteDeclaration(None=LandscapeLayerSample() {parameterName=BiomeA}) {name=BiomeA_Weight, nodeColor=<rgba 0.5 0 1 1>}
  n59 = Constant()
  n60 = NamedRerouteUsage()
  n61 = ScalarParameter() {defaultValue=-2, group=GrassMeshes, parameterName=GrassLayer_01_SpawnLA}
  n62 = ScalarParameter() {defaultValue=1.5, group=GrassMeshes, parameterName=GrassLayer_01_SpawnLB}
  n63 = NamedRerouteUsage()
  n64 = NamedRerouteUsage()
  n65 = NamedRerouteUsage()
  n66 = NamedRerouteUsage()
  n67 = NamedRerouteUsage()
  n68 = Saturate(None=Multiply(A=Max(A=@n66, B=@n67), B=Constant() {r=3}))
  n69 = NamedRerouteUsage()
  n70 = NamedRerouteUsage()
  n71 = Max(A=Max(A=Max(A=@n65, B=@n68), B=Max(A=@n69, B=@n70)), B=NamedRerouteUsage())
  n72 = Multiply(A=Saturate(None=LinearInterpolate(A=@n61, B=@n62, Alpha=@n63)), B=OneMinus(None=Max(A=@n64, B=@n71)))
  MaterialExpressionReroute_6 = Reroute(None=@n72)
  n73 = MaterialFunctionCall(In=@MaterialExpressionReroute_6.Output) {materialFunction=&MF_LS_GrassSwitch}
  n74 = NamedRerouteUsage()
  n75 = StaticSwitch(True=Multiply(A=@n60, B=@n73.Result), False=@n73.Result, Value=@n74)
  n76 = NamedRerouteUsage()
  NamedRerouteDeclaration(None=StaticSwitch(True=@n59, False=@n75, Value=@n76)) {name=Shared_Grass_Layer01, nodeColor=<rgba 0.5 0 1 1>}
  n77 = Max(A=Max(A=NamedRerouteUsage(), B=NamedRerouteUsage()), B=NamedRerouteUsage())
  n78 = NamedRerouteUsage()
  n79 = Multiply(A=@n10, B=OneMinus(None=Max(A=@n77, B=@n78)))
  MaterialExpressionReroute_13 = Reroute(None=@n79)
  NamedRerouteDeclaration(None=@MaterialExpressionReroute_13.Output) {name=Dirt_Physmat_WeightMap, nodeColor=<rgba 0.5 0 1 1>}
  NamedRerouteDeclaration(None=@n28.R) {name=Biome_Chaparral_Coast, desc=Acronym:CHPC, nodeColor=<rgba 0.5 0 1 1>}
  n80 = NamedRerouteUsage()
  NamedRerouteDeclaration(None=Saturate(None=Subtract(A=@n28.G, B=@n80))) {name=Biome_Boreal_Forest, desc=Acronym:BRLF, nodeColor=<rgba 0.5 0 1 1>}
  n81 = NamedRerouteUsage()
  NamedRerouteDeclaration(None=Saturate(None=Subtract(A=@n28.B, B=@n81))) {name=Biome_Boreal_Tundra, desc=Acronym:BRLT, nodeColor=<rgba 0.5 0 1 1>}
  n82 = NamedRerouteUsage()
  NamedRerouteDeclaration(None=Saturate(None=Subtract(A=@n28.A, B=@n82))) {name=Biome_Boreal_Mountain, desc=Acronym:BRLM, nodeColor=<rgba 0.5 0 1 1>}
  n83 = Saturate(None=LinearInterpolate(
    A=ScalarParameter() {defaultValue=-2, group=GrassMeshes, parameterName=ScotchBroomLayer_01_SpawnLA}
    B=ScalarParameter() {defaultValue=1.5, group=GrassMeshes, parameterName=ScotchBroomLayer_01_SpawnLB}
    Alpha=NamedRerouteUsage()
   ))
  n84 = NamedRerouteUsage()
  n85 = Max(A=NamedRerouteUsage(), B=NamedRerouteUsage())
  n86 = Constant() {r=3}
  n87 = Max(A=NamedRerouteUsage(), B=Saturate(None=Multiply(A=@n85, B=@n86)))
  n88 = NamedRerouteUsage()
  n89 = Multiply(A=Multiply(A=Multiply(A=@n83, B=@n84), B=NamedRerouteUsage()), B=OneMinus(None=Max(A=@n87, B=@n88)))
  MaterialExpressionReroute_5 = Reroute(None=@n89)
  NamedRerouteDeclaration(None=MaterialFunctionCall(In=@MaterialExpressionReroute_5.Output) {materialFunction=&MF_LS_GrassSwitch}.Result) {name=CHPC_Layer01, nodeColor=<rgba 0.5 0 1 1>}
  MaterialExpressionReroute_12 = Reroute(None=@n6)
  NamedRerouteDeclaration(None=@MaterialExpressionReroute_12.Output) {name=Stone_Physmat_weightmap, nodeColor=<rgba 0.5 0 1 1>}
  n90 = LinearInterpolate(
   A=ScalarParameter() {group=GrassMeshes, parameterName=Chaparral_GrassBreakup_LA}
   B=ScalarParameter() {defaultValue=2, group=GrassMeshes, parameterName=Chaparral_GrassBreakup_LB}
   Alpha=MaterialFunctionCall!(
    Coordinates
    R_01_Scale=ScalarParameter() {defaultValue=10000, group=GrassMeshes, parameterName=Chaparral_GrassBreakup}
    R_02_Scale
    R_03_Scale
    R_04_Scale
    G_01_Scale
    G_02_Scale
    G_03_Scale
    G_04_Scale
   ) {materialFunction=/Game/Athena/Helios/Environments/Landscape/Materials/MaterialFunctions_Utilities/MF_VariationMasks2}.R_01
  )
  NamedRerouteDeclaration(None=LinearInterpolate(B=Saturate(None=@n90), Alpha=NamedRerouteUsage())) {name=ChaparralGrassBreakup_Noise, nodeColor=<rgba 0.5 0 1 1>}
  n91 = LinearInterpolate(
   A=ScalarParameter() {defaultValue=-2, group=GrassMeshes, parameterName=ChaparralGrassLayer_01_SpawnLA}
   B=ScalarParameter() {defaultValue=1, group=GrassMeshes, parameterName=ChaparralGrassLayer_01_SpawnLB}
   Alpha=NamedRerouteUsage()
  )
  n92 = NamedRerouteUsage()
  n93 = Max(A=NamedRerouteUsage(), B=NamedRerouteUsage())
  n94 = Constant() {r=3}
  n95 = Max(A=NamedRerouteUsage(), B=Saturate(None=Multiply(A=@n93, B=@n94)))
  n96 = Multiply(A=Multiply(A=Saturate(None=@n91), B=NamedRerouteUsage()), B=OneMinus(None=Max(A=@n92, B=@n95)))
  MaterialExpressionReroute_7 = Reroute(None=@n96)
  n97 = MaterialFunctionCall(In=@MaterialExpressionReroute_7.Output) {materialFunction=&MF_LS_GrassSwitch}
  NamedRerouteDeclaration(None=StaticSwitch(True=Multiply(A=@n60, B=@n97.Result), False=@n97.Result, Value=@n74)) {name=CHPG_Layer01, nodeColor=<rgba 0.5 0 1 1>}
  n98 = MaterialFunctionCall!(
   Material=NamedRerouteUsage()
   MobileMaterial=NamedRerouteUsage()
   GrassWeight
   LayerWeight=NamedRerouteUsage()
   UseTangentNormals
   TerrainMasks_QuadA=NamedRerouteUsage()
   TerrainMasks_QuadB=NamedRerouteUsage()
   TerrainMasks_QuadC=NamedRerouteUsage()
   TerrainMasks_QuadD=NamedRerouteUsage()
   VariationMasks Mobile=NamedRerouteUsage()
  ) {materialFunction=/Game/Athena/Helios/Environments/Landscape/Materials/LandscapeLayers/MF_Helios_Grass_BorealForest}
  NamedRerouteDeclaration(None=@n98.MobileLayerMaterial) {name=MobileBorealForestMaterial, nodeColor=<rgba 0.4766 1 0 1>}
  n99 = Saturate(None=Multiply(A=NamedRerouteUsage(), B=Constant() {r=3}))
  n100 = Max(A=NamedRerouteUsage(), B=NamedRerouteUsage())
  n101 = NamedRerouteUsage()
  n102 = Max(A=Max(A=Max(A=@n100, B=@n101), B=NamedRerouteUsage()), B=NamedRerouteUsage())
  n103 = NamedRerouteUsage()
  n104 = NamedRerouteUsage()
  n105 = Constant() {r=6}
  n106 = Max(A=Max(A=Max(A=@n102, B=@n103), B=NamedRerouteUsage()), B=Saturate(None=Multiply(A=@n104, B=@n105)))
  n107 = Saturate(None=Multiply(A=OneMinus(None=@n106), B=Constant() {r=6}))
  n108 = MaterialFunctionCall(In=ShadingPathSwitch(Default=Multiply(A=@n99, B=@n107), Mobile=Constant())) {materialFunction=&MF_LS_GrassSwitch}
  n109 = Multiply(A=@n60, B=@n108.Result)
  n110 = NamedRerouteUsage()
  NamedRerouteDeclaration(None=Multiply(A=StaticSwitch(True=@n109, False=@n108.Result, Value=@n74), B=OneMinus(None=@n110))) {name=Shared_Gravel_Pebbles01, nodeColor=<rgba 0.5 0 1 1>}
  NamedRerouteDeclaration(None=@n21.CliffMaterialOnly) {name=CliffMaterialOnly, nodeColor=<rgba 0.6172 0 1 1>}
  n111 = Max(A=Max(A=NamedRerouteUsage(), B=NamedRerouteUsage()), B=NamedRerouteUsage())
  n112 = Constant() {desc=Replace With SnowLayer_Alpha}
  n113 = Multiply(A=Max(A=NamedRerouteUsage(), B=NamedRerouteUsage()), B=OneMinus(None=Max(A=@n111, B=@n112)))
  MaterialExpressionReroute_11 = Reroute(None=@n113)
  NamedRerouteDeclaration(None=@MaterialExpressionReroute_11.Output) {name=Grass_Physmat_Weightmap, nodeColor=<rgba 0.5 0 1 1>}
  n114 = NamedRerouteUsage()
  n115 = Constant() {desc=Replace With SnowLayer_Alpha}
  n116 = Multiply(A=NamedRerouteUsage(), B=OneMinus(None=Max(A=@n114, B=@n115)))
  MaterialExpressionReroute_10 = Reroute(None=@n116)
  NamedRerouteDeclaration(None=@MaterialExpressionReroute_10.Output) {name=Sand_Physmat_Weightmap, nodeColor=<rgba 0.5 0 1 1>}
  n117 = NamedRerouteUsage()
  n118 = NamedRerouteUsage()
  n119 = Multiply(A=Constant() {r=2}, B=Multiply(A=OneMinus(None=@n117), B=OneMinus(None=@n118)))
  MaterialExpressionReroute_9 = Reroute(None=@n119)
  NamedRerouteDeclaration(None=@MaterialExpressionReroute_9.Output) {name=Snow_physmat_Weightmap, nodeColor=<rgba 0.5 0 1 1>}
  n120 = Multiply(A=@n98.BorealForest_GrassMask, B=ScalarParameter() {defaultValue=4, group=GrassMeshes, parameterName=Grass_BorealForest_MeshMask_Pow})
  n121 = ScalarParameter() {defaultValue=-0.75, group=GrassMeshes, parameterName=Grass_BorealForest_GrassMeshMask_Add}
  NamedRerouteDeclaration(None=Saturate(None=Add(A=@n120, B=@n121))) {name=BorealForest_GrassMask, nodeColor=<rgba 0.5 0 1 1>}
  NamedRerouteDeclaration(None=StaticSwitch(True=@n75, False=@n59, Value=@n76)) {name=Lobby_Grass_Layer01, nodeColor=<rgba 0.5 0 1 1>}
  NamedRerouteDeclaration(None=@n38) {name=Grass_EdgeOnly, nodeColor=<rgba 0.5 0 1 1>}
  n122 = NamedRerouteDeclaration(None=StaticBoolParameter() {parameterName=UseGrass_BorealForest}) {name=UseBorealForest, nodeColor=<rgba 0.5 0 1 1>}
  NamedRerouteDeclaration(None=StaticSwitch(True=@n98.BlendedMaterial, False=NamedRerouteUsage(), Value=@n122)) {name=Layer0_Material, nodeColor=<rgba 0.5 0 1 1>}
  n123 = OneMinus(None=NamedRerouteUsage())
  NamedRerouteDeclaration(None=StaticSwitch(
    True=MaterialFunctionCall(In=@n123) {materialFunction=&MF_LS_GrassSwitch}.Result
    False=Constant()
    Value=@n74
   )) {name=BRL_Snow, nodeColor=<rgba 0.5 0 1 1>}
  n124 = Saturate(None=LinearInterpolate(A=Constant() {r=-2}, B=Constant() {r=1}, Alpha=NamedRerouteUsage()))
  n125 = Max(A=NamedRerouteUsage(), B=NamedRerouteUsage())
  n126 = MaterialFunctionCall() {materialFunction=/Game/Athena/Helios/Environments/Landscape/Materials/MaterialFunctions_Utilities/MF_ScotchBroom_Colors}
  n127 = NamedRerouteUsage()
  n128 = NamedRerouteUsage()
  n129 = Saturate(None=Multiply(A=Max(A=@n127, B=@n128), B=Constant() {r=3}))
  n130 = NamedRerouteUsage()
  n131 = Multiply(
   A=Multiply(A=Multiply(A=@n124, B=@n125), B=Power(Base=@n126.LandscapeLayerMaskNoise) {constExponent=10})
   B=OneMinus(None=Max(A=@n129, B=@n130))
  )
  MaterialExpressionReroute_0 = Reroute(None=@n131)
  n132 = MaterialFunctionCall(In=@MaterialExpressionReroute_0.Output) {materialFunction=&MF_LS_GrassSwitch}
  NamedRerouteDeclaration(None=StaticSwitch(True=Multiply(A=@n60, B=@n132.Result), False=@n132.Result, Value=@n74)) {name=BRL_Flowers, nodeColor=<rgba 0.5 0 1 1>}
  n133 = NamedRerouteUsage()
  n134 = MaterialFunctionCall(In=Multiply(A=NamedRerouteUsage(), B=OneMinus(None=@n133))) {materialFunction=&MF_LS_GrassSwitch}
  NamedRerouteDeclaration(None=StaticSwitch(True=Multiply(A=@n60, B=@n134.Result), False=@n134.Result, Value=@n74)) {name=Shared_Scree, nodeColor=<rgba 0.5 0 1 1>}
  n135 = LandscapeLayerSample() {desc=AridDesert, parameterName=Temp5}
  n136 = LandscapeLayerSample() {desc=OpenDesert, parameterName=Temp6}
  n137 = Constant()
  n138 = MaterialFunctionCall(X=@n135, Y=@n136, Z=@n137, A=@n137) {materialFunction=&MakeFloat4}
  n139 = MaterialFunctionCall(Float4=NamedRerouteDeclaration(None=StaticSwitchParameter(True=@n138.Result, False=@n24.QuadD_RGBA) {group=Temp, parameterName=UseTempLayersForBiomeMasks-HadesOlympus}) {name=TerrainMasks_QuadD, nodeColor=<rgba 0.5 0 1 1>}) {materialFunction=&BreakOutFloat4Components}
  NamedRerouteDeclaration(None=@n139.R) {name=Biome_Hades, desc=Hades, nodeColor=<rgba 0.5 0 1 1>}
  NamedRerouteDeclaration(None=@n139.G) {name=Biome_Olympus, desc=Olympus, nodeColor=<rgba 0.5 0 1 1>}
  NamedRerouteDeclaration(None=StaticSwitch(True=@n98.MobileBlendedMaterial, False=NamedRerouteUsage(), Value=@n122)) {name=Layer0_MobileMaterial, nodeColor=<rgba 0.5 0 1 1>}
  n140 = NamedRerouteDeclaration(None=StaticBoolParameter() {parameterName=UseGrass_ChaparralCoast}) {name=UseChaparralCoast, nodeColor=<rgba 0.5 0 1 1>}
  NamedRerouteDeclaration(None=StaticSwitch(True=@n31.BlendedMaterial, False=NamedRerouteUsage(), Value=@n140)) {name=Layer0_ChapCoastGrass, nodeColor=<rgba 0.5 0 1 1>}
  NamedRerouteDeclaration(None=StaticSwitch(True=@n31.MobileBlendedMaterial, False=NamedRerouteUsage(), Value=@n140)) {name=Layer0_ChapCoastGrassMobile, nodeColor=<rgba 0.5 0 1 1>}
  NamedRerouteDeclaration(None=MaterialFunctionCall(LayerWEight=@n21.LayerMask, Material=@n21.LayerMask) {materialFunction=&MF_WastelandColorTint}.Material) {name=AllLayersWithCliff, nodeColor=<rgba 0.6172 0 1 1>}
  NamedRerouteDeclaration(None=LandscapeLayerSample() {parameterName=BiomeC}) {name=BiomeC_Weight, nodeColor=<rgba 0.5 0 1 1>}
  NamedRerouteDeclaration(None=@n21.MobileBlendedMaterial) {name=AllLayersWithCliff_Mobile, nodeColor=<rgba 0.6172 0 1 1>}
  NamedRerouteDeclaration(None=StaticSwitch(True=NamedRerouteUsage(), False=Constant(), Value=@n23)) {name=SnowMossDesat, nodeColor=<rgba 0.6172 0 1 1>}
  NamedRerouteDeclaration(None=@n20.BlendedMaterial) {name=MaterialUpTo_Asphalt, nodeColor=<rgba 0.5 0 1 1>}
  NamedRerouteDeclaration(None=@n20.MobileBlendedMaterial) {name=MobileMaterialUpTo_Asphalt, nodeColor=<rgba 0.5 0 1 1>}
  NamedRerouteDeclaration(None=StaticSwitch(True=@n50.BlendedMaterial, False=NamedRerouteUsage(), Value=@n51)) {name=MaterialUpTo_ForestFloor, nodeColor=<rgba 0.5 0 1 1>}
  NamedRerouteDeclaration(None=StaticSwitch(True=@n50.MobileBlendedMaterial, False=NamedRerouteUsage(), Value=@n51)) {name=MobileMaterialUpTo_ForestFloor, nodeColor=<rgba 0.5 0 1 1>}
  n141 = MaterialFunctionCall!(
   Material=NamedRerouteUsage()
   MobileMaterial=NamedRerouteUsage()
   UseTangentNormals
   LayerWeight=NamedRerouteUsage()
   TerrainMasks_QuadA=NamedRerouteUsage()
   TerrainMasks_QuadB=NamedRerouteUsage()
   TerrainMasks_QuadC=NamedRerouteUsage()
   TerrainMasks_QuadD=NamedRerouteUsage()
   VariationMasks Mobile=NamedRerouteUsage()
  ) {materialFunction=/Game/Athena/Helios/Environments/Landscape/Materials/LandscapeLayers/MF_Helios_Brimstone01}
  n142 = NamedRerouteDeclaration(None=StaticBoolParameter() {group=BiomeABCD_Settings, parameterName=UseBiomeA_Brimstone01, sortPriority=-500}) {name=UseBiomeA_Brimstone01, nodeColor=<rgba 0.5 0 1 1>}
  NamedRerouteDeclaration(None=StaticSwitch(True=@n141.LayerMask, False=Constant(), Value=@n142)) {name=Brimstone01LayerMask, nodeColor=<rgba 0.5 0 1 1>}
  NamedRerouteDeclaration(None=StaticSwitch(True=@n141.BlendedMaterial, False=NamedRerouteUsage(), Value=@n142)) {name=MaterialUpTo_Brimstone01, nodeColor=<rgba 0.5 0 1 1>}
  NamedRerouteDeclaration(None=LandscapeLayerSample() {parameterName=BiomeD}) {name=BiomeD_Weight, nodeColor=<rgba 0.5 0 1 1>}
  NamedRerouteDeclaration(None=StaticSwitch(True=@n141.MobileBlendedMaterial, False=NamedRerouteUsage(), Value=@n142)) {name=MobileMaterialUpTo_Brimstone01, nodeColor=<rgba 0.5 0 1 1>}
  n143 = Saturate(None=LinearInterpolate(A=Constant() {r=-2}, B=Constant() {r=1.5}, Alpha=NamedRerouteUsage()))
  n144 = NamedRerouteUsage()
  n145 = Multiply(A=Max(A=NamedRerouteUsage(), B=NamedRerouteUsage()), B=Constant() {r=3})
  n146 = Max(A=Max(A=NamedRerouteUsage(), B=Saturate(None=@n145)), B=NamedRerouteUsage())
  n147 = MaterialFunctionCall(In=Multiply(A=Multiply(A=@n143, B=@n144), B=OneMinus(None=@n146))) {materialFunction=&MF_LS_GrassSwitch}
  NamedRerouteDeclaration(None=StaticSwitch(True=Multiply(A=@n60, B=@n147.Result), False=@n147.Result, Value=@n74)) {name=SunGold01, nodeColor=<rgba 0.5 0 1 1>}
  n148 = NamedRerouteUsage()
  n149 = Multiply(A=Max(A=NamedRerouteUsage(), B=NamedRerouteUsage()), B=Constant() {r=3})
  n150 = Max(A=Max(A=NamedRerouteUsage(), B=NamedRerouteUsage()), B=Max(A=NamedRerouteUsage(), B=NamedRerouteUsage()))
  n151 = Max(A=Max(A=NamedRerouteUsage(), B=NamedRerouteUsage()), B=Max(A=NamedRerouteUsage(), B=NamedRerouteUsage()))
  n152 = OneMinus(None=Max(A=Saturate(None=@n149), B=Max(A=@n150, B=@n151)))
  NamedRerouteDeclaration(None=MaterialFunctionCall(In=Multiply(A=@n148, B=@n152)) {materialFunction=&MF_LS_GrassSwitch}.Result) {name=WastelandShrubMask, nodeColor=<rgba 0.0078 0 1 1>}
  NamedRerouteDeclaration(None=@n136) {name=Desert_weight, nodeColor=<rgba 0.9219 0 1 1>}
  NamedRerouteDeclaration(None=@n135) {name=Wasteland_wieght, nodeColor=<rgba 0.5234 1 0 1>}
  NamedRerouteDeclaration(None=@n14.LayerMask) {name=WastelandLayerMask, nodeColor=<rgba 0.9453 1 0 1>}
  NamedRerouteDeclaration(None=StaticSwitch(
    True=MaterialFunctionCall(LayerWEight=@n39.LayerMask, Material=@n39.LayerMask) {materialFunction=&MF_WastelandColorTint}.Material
    False=NamedRerouteUsage()
    Value=@n40
   )) {name=MaterialUpTo_Scree, nodeColor=<rgba 0.5 0 1 1>}
  NamedRerouteDeclaration(None=StaticSwitch(True=@n39.MobileBlendedMaterial, False=NamedRerouteUsage(), Value=@n40)) {name=MobileMaterialUpTo_Scree, nodeColor=<rgba 0.5 0 1 1>}
  n153 = MaterialFunctionCall!(
   Material=NamedRerouteUsage()
   MobileMaterial=NamedRerouteUsage()
   Mobile Sand Material
   LayerWeight=NamedRerouteUsage()
   UseTangentNormals
   TerrainMasks_QuadA=NamedRerouteUsage()
   TerrainMasks_QuadB=NamedRerouteUsage()
   TerrainMasks_QuadC=NamedRerouteUsage()
   TerrainMasks_QuadD=NamedRerouteUsage()
   VariationMasks Mobile=NamedRerouteUsage()
  ) {materialFunction=/Game/Athena/Helios/Environments/Landscape/Materials/LandscapeLayers/MF_Helios_Brimstone02}
  n154 = NamedRerouteDeclaration(None=StaticBoolParameter() {group=BiomeABCD_Settings, parameterName=UseBiomeC_Brimstone02, sortPriority=-500}) {name=UseBiomeC_Brimstone02, nodeColor=<rgba 0.5 0 1 1>}
  NamedRerouteDeclaration(None=StaticSwitch(True=@n153.LayerMask, False=Constant(), Value=@n154)) {name=Brimstone02LayerMask, nodeColor=<rgba 0.5 0 1 1>}
  n155 = NamedRerouteDeclaration(None=MaterialFunctionCall() {materialFunction=/Game/Athena/Helios/Environments/Landscape/Materials/MaterialFunctions_Utilities/MF_CliffSlopeMask}.Result) {name=Cliff_Weight, nodeColor=<rgba 0.5 0 1 1>}
  NamedRerouteDeclaration(None=OneMinus(None=Max(A=@n38, B=@n155))) {name=Grass_Weight, nodeColor=<rgba 0.5 0 1 1>}
  NamedRerouteDeclaration(None=StaticSwitch(True=@n153.BlendedMaterial, False=NamedRerouteUsage(), Value=@n154)) {name=MaterialUpTo_Brimstone02, nodeColor=<rgba 0.5 0 1 1>}
  NamedRerouteDeclaration(None=StaticSwitch(True=@n153.MobileBlendedMaterial, False=NamedRerouteUsage(), Value=@n154)) {name=MobileMaterialUpTo_Brimstone02, nodeColor=<rgba 0.5 0 1 1>}
  NamedRerouteDeclaration(None=@n19.LayerMask) {name=DesertLayerMask, nodeColor=<rgba 0 0.0391 1 1>}
  NamedRerouteUsage()
  NamedRerouteUsage()
  NamedRerouteUsage()
  NamedRerouteUsage()
  NamedRerouteUsage()
  NamedRerouteUsage()
  NamedRerouteUsage()
  NamedRerouteUsage()
  NamedRerouteUsage()
  NamedRerouteUsage()
  NamedRerouteUsage()
  OneMinus(None=@n16)
  PinBase()
  PinBase!(Output=@n131) {reroutePins [Output {expression=&self:MaterialExpressionReroute_0}]}
  PinBase!()
  PinBase!(Output=@n119) {reroutePins [Output {expression=&self:MaterialExpressionReroute_9}]}
  PinBase!()
  PinBase!(Output=@n96) {reroutePins [Output {expression=&self:MaterialExpressionReroute_7}]}
  PinBase!()
  PinBase!(Output=@n42) {reroutePins [Output {expression=&self:MaterialExpressionReroute_1}]}
  PinBase!()
  PinBase!(Output=@n116) {reroutePins [Output {expression=&self:MaterialExpressionReroute_10}]}
  PinBase!()
  PinBase!(Output=@n113) {reroutePins [Output {expression=&self:MaterialExpressionReroute_11}]}
  PinBase!()
  PinBase!(Output=@n6) {reroutePins [Output {expression=&self:MaterialExpressionReroute_12}]}
  PinBase!()
  PinBase!(Output=@n79) {reroutePins [Output {expression=&self:MaterialExpressionReroute_13}]}
  PinBase!() {reroutePins [
    "HighEndMaterial(MA)" {expression=&self:MaterialExpressionReroute_20}
    "MobileMaterial(MA)" {expression=&self:MaterialExpressionReroute_21}
   ]}
  n156 = RuntimeVirtualTextureSampleParameter() {
   group=RVT
   materialType=BaseColor_Normal_Specular_YCoCg
   parameterName=RVT Sample Target
   virtualTexture=/Game/Athena/Apollo/Maps/Landscape/VT/RVT_Apollo_Landscape
  }
  n157 = RuntimeVirtualTextureSampleParameter() {
   group=RVT
   materialType=Displacement
   parameterName=RVT Displacement
   virtualTexture=/Game/Athena/Apollo/Maps/Landscape/VT/RVT_Helios_Displacement
  }
  n158 = SetMaterialAttributes!(
   MaterialAttributes
   Base Color=@n156.BaseColor
   Specular=@n156.BaseColor
   Roughness=@n156.BaseColor
   Normal=@n156.BaseColor
   Displacement=@n157.Displacement
  ) {attributeSetTypes [
    #3
    #2
    #4
    #5
    #6
   ]}
  MaterialExpressionReroute_21 = Reroute(None=@n2)
  n159 = MaterialFunctionCall!(MobileMaterial=@MaterialExpressionReroute_21."MobileMaterial(MA)", TransformTangents) {materialFunction=/Game/Athena/Helios/Environments/Landscape/Materials/MaterialFunctions_Utilities/MF_MobileDeriveNormals}
  n160 = ShadingPathSwitch(Default=VirtualTextureFeatureSwitch!(No=@n159.ResultMaterial, Yes=@n158), Mobile=@n159.ResultMaterial)
  PinBase!(Output=@n160) {reroutePins [Output {expression=&self:MaterialExpressionReroute_18}]}
  PinBase!()
  PinBase!(Output=@n57) {reroutePins [Output {expression=&self:MaterialExpressionReroute_4}]}
  PinBase!()
  PinBase!(Output=@n89) {reroutePins [Output {expression=&self:MaterialExpressionReroute_5}]}
  PinBase!()
  PinBase!(Output=@n72) {reroutePins [Output {expression=&self:MaterialExpressionReroute_6}]}
  MaterialExpressionReroute_18 = Reroute(None=@n160)
  n161 = MaterialFunctionCall!(
   Material=@MaterialExpressionReroute_18.Output
   UseSnowForBiomeC=NamedRerouteUsage()
   Snow_PostRVT=NamedRerouteUsage()
   SwitchGrassLayerMask
   TerrainMasks_QuadA
   TerrainMasks_QuadB
   TerrainMasks_QuadC
   TerrainMasks_QuadD
  ) {
   desc=For Post RVT Details on landscape layers
   materialFunction=/Game/Athena/Helios/Environments/Landscape/Materials/MaterialFunctions_Utilities/PostRVT_LayerDetails
  }
  n162 = MaterialFunctionCall!(LandscapeMaterial=@n161.Result, VariationMasksMobile) {
   desc=For Common Gameplay/Misc Post RVT Details
   materialFunction=/Game/Athena/Helios/Environments/Landscape/Materials/MaterialFunctions_Utilities/MF_PostRVT_Misc
  }
  n163 = GetMaterialAttributes(None=@n162.LandscapeMaterial) {attributeGetTypes [#3]}
  Reroute(None=BlendMaterialAttributes(
    A=@n162.LandscapeMaterial
    B=SetMaterialAttributes!(MaterialAttributes=@n163.MaterialAttributes, Base Color) {attributeSetTypes [#3]}
    Alpha=ScalarParameter() {parameterName=DEBUG_AsBaseColor}
   ))
  MaterialExpressionReroute_20 = Reroute(None=@n1)
  n164 = GetMaterialAttributes(None=ShadingPathSwitch(Default=@MaterialExpressionReroute_20."HighEndMaterial(MA)", Mobile=SetMaterialAttributes!(MaterialAttributes))) {attributeGetTypes [
    #3
    #2
    #4
    #5
    #6
   ]}
  RuntimeVirtualTextureOutput!(
   BaseColor=MaterialFunctionCall!(BaseColor=@n164.Base Color) {materialFunction=/Game/Athena/Apollo/Environments/Landscape/Materials/Biome/CommonMaterials/MF_ReduceRTV-CompressionArtifacts}.BaseColor
   Specular=@n164.Specular
   Roughness=@n164.Specular
   Normal=@n164.Specular
   WorldHeight=WorldPosition().Z
   Opacity
   Mask
   Displacement=Multiply(A=@n164.Displacement, B=ScalarParameter() {defaultValue=1, group=Displacement, parameterName=Displacement_BASE_Mod})
   Mask4
  )
  n165 = GetMaterialAttributes(None=@n162.LandscapeMaterial) {attributeGetTypes [#3]}
  n166 = VectorParameter() {group=Debug, parameterName=DEBUG-TraversalMaskColor, defaultValue=<rgba 1 0 0 1>}
  n167 = ScalarParameter() {desc=Set this to 1 to turn it on, group=Debug, parameterName=DEBUG-TraversalMask}
  n168 = Constant() {r=43}
  n169 = Constant() {r=-16.5}
  n170 = WorldPosition()
  n171 = CrossProduct(A=DDX(Value=@n170.XYZ), B=DDY(Value=@n170.XYZ))
  n172 = Abs(None=ComponentMask(None=Normalize(VectorInput=@n171)) {b=true})
  n173 = Constant() {r=1}
  n174 = If(A=Saturate(None=LinearInterpolate(A=@n168, B=@n169, Alpha=@n172)), B=Constant() {r=0.75}, "A > B"=@n173, "A == B"=@n173, "A < B"=Constant())
  SetMaterialAttributes!(
   MaterialAttributes=@n165.MaterialAttributes
   Base Color=LinearInterpolate(A=@n165.Base Color, B=Multiply(A=@n165.Base Color, B=@n166.RGB), Alpha=Multiply(A=@n167, B=@n174))
  ) {attributeSetTypes [#3]}
  StaticSwitchParameter(
   True=MaterialFunctionCall(
    X=Constant() {desc=Temp1 Chaparral Coast}
    Y=Constant() {desc=Temp3 Boreal Forest}
    Z=Constant() {desc=Temp2 Boreal Tundra}
    A=Constant() {desc=Temp4 Boreal Mountain}
   ) {materialFunction=&MakeFloat4}.Result
   False=@n24.QuadB_RGBA
  ) {defaultValue=true, group=Temp, parameterName=UseTempLayersForBiomeMasks}
  StaticSwitch(True=Multiply(A=@n60, B=@n43.Result), False=@n43.Result, Value=@n74)
  StaticSwitch(True=Multiply(A=@n60, B=@n58.Result), False=@n58.Result, Value=@n74)
 ]
}
